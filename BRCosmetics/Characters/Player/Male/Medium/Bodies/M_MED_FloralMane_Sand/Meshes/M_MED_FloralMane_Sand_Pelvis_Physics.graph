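# Physics asset for the pelvis mesh of the M_MED_FloralMane_Sand male player cosmetic, setting up secondary motion for shirt, drawstrings and drawstring loops.
# Capsule bodies form left and right chains from MSTR anchor bones, alongside unconstrained pelvis and thigh capsules.
# Angular limits keep shirt chains stiffest, loops slightly looser and drawstrings loosest.
# Drawstring and loop chains attach to the opposite side's MSTR bone, unlike the shirt chains.
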PhysicsAsset /BRCosmetics/Characters/Player/Male/Medium/Bodies/M_MED_FloralMane_Sand/Meshes/M_MED_FloralMane_Sand_Pelvis_Physics
bodies {
 dyn_drawstring_l_01 [{shapeType=Capsule, length=6.5757, radius=1.4243, shapeName=dyn_drawstring_l_01_capsule, center=<1.8493 0 2.7>, rotation=<rot 90 0 0>}]
 dyn_drawstring_l_02 [{
   shapeType=Capsule
   length=6.5479
   radius=1.4521
   shapeName=dyn_drawstring_l_02_capsule
   center=<1.7482 0 2.2>
   rotation=<rot 79.7878 -0.0001 -0.0001>
  }]
 dyn_drawstring_l_03 [{
   shapeType=Capsule
   length=6.8839
   radius=1.1161
   shapeName=dyn_drawstring_l_03_capsule
   center=<1.9669 0 1.26>
   rotation=<rot 81.748 -0.0067 -0.0051>
  }]
 dyn_drawstring_l_MSTR [{shapeType=Capsule, length=0.505, radius=0.505, shapeName=dyn_drawstring_l_MSTR_capsule, rotation=<rot 0 -90 0>}]
 dyn_drawstring_loop_l_01 [{shapeType=Capsule, length=2.8632, radius=2.9916, shapeName=dyn_drawstring_loop_l_01_capsule, center=<2.687 3.6 0>, rotation=<rot 90 0 0>}]
 dyn_drawstring_loop_l_02 [{shapeType=Capsule, length=2.8632, radius=2.9916, shapeName=dyn_drawstring_loop_l_02_capsule, center=<2.687 3.4 0>, rotation=<rot 90 0 0>}]
 dyn_drawstring_loop_l_MSTR [{shapeType=Capsule, length=0.505, radius=0.505, shapeName=dyn_drawstring_loop_l_MSTR_capsule, rotation=<rot 0 -90 0>}]
 dyn_drawstring_loop_r_01 [{shapeType=Capsule, length=2.9432, radius=2.9116, shapeName=dyn_drawstring_loop_r_01_capsule, center=<2.687 -3.8 0>, rotation=<rot 90 0 0>}]
 dyn_drawstring_loop_r_02 [{shapeType=Capsule, length=2.9432, radius=2.9116, shapeName=dyn_drawstring_loop_r_02_capsule, center=<2.687 -3.8 0>, rotation=<rot 90 0 0>}]
 dyn_drawstring_loop_r_MSTR [{shapeType=Capsule, length=0.505, radius=0.505, shapeName=dyn_drawstring_loop_r_MSTR_capsule, rotation=<rot 0 -90 0>}]
 dyn_drawstring_r_01 [{
   shapeType=Capsule
   length=6.5757
   radius=1.4243
   shapeName=dyn_drawstring_r_01_capsule
   center=<1.9159 0 1.95>
   rotation=<rot 86.5509 0.0001 0.0001>
  }]
 dyn_drawstring_r_02 [{shapeType=Capsule, length=6.5015, radius=1.4985, shapeName=dyn_drawstring_r_02_capsule, center=<1.9326 0 1.5>, rotation=<rot 79.7932 0 0>}]
 dyn_drawstring_r_03 [{
   shapeType=Capsule
   length=6.1452
   radius=1.8548
   shapeName=dyn_drawstring_r_03_capsule
   center=<2.0159 0 1.2>
   rotation=<rot 79.7119 0.0001 0>
  }]
 dyn_drawstring_r_MSTR [{shapeType=Capsule, length=0.505, radius=0.505, shapeName=dyn_drawstring_r_MSTR_capsule, rotation=<rot 0 -90 0>}]
 dyn_shirt_l_01 [{
   shapeType=Capsule
   length=13.5816
   radius=4.244
   shapeName=dyn_shirt_l_01_capsule
   center=<5.773 0 4.7>
   rotation=<rot 0.0005 0.0004 85.5138>
  }]
 dyn_shirt_l_02 [{
   shapeType=Capsule
   length=13.5901
   radius=4.2356
   shapeName=dyn_shirt_l_02_capsule
   center=<5.773 0 5.05>
   rotation=<rot 0.0005 0.0004 85.6972>
  }]
 dyn_shirt_l_03 [{
   shapeType=Capsule
   length=13.5816
   radius=4.244
   shapeName=dyn_shirt_l_03_capsule
   center=<5.773 0 4.8>
   rotation=<rot 0.0005 0.0004 92.2083>
  }]
 dyn_shirt_l_04 [{
   shapeType=Capsule
   length=13.5816
   radius=4.244
   shapeName=dyn_shirt_l_04_capsule
   center=<5.773 0 4.8>
   rotation=<rot 0.0004 0.0003 98.8059>
  }]
 dyn_shirt_l_MSTR [{shapeType=Capsule, length=0.4366, radius=0.749, shapeName=dyn_shirt_l_MSTR_capsule, center=<-0.1951 -1.4333 0>, rotation=<rot 0 -90 0>}]
 dyn_shirt_r_01 [{shapeType=Capsule, length=13.3359, radius=4.4897, shapeName=dyn_shirt_l_01_capsule, center=<5.773 0 4.7>, rotation=<rot 0 0 90>}]
 dyn_shirt_r_02 [{
   shapeType=Capsule
   length=13.3613
   radius=4.4643
   shapeName=dyn_shirt_l_02_capsule
   center=<5.773 -0.0061 5.1143>
   rotation=<rot 0.0004 -0.0003 86.9786>
  }]
 dyn_shirt_r_03 [{
   shapeType=Capsule
   length=13.3952
   radius=4.4305
   shapeName=dyn_shirt_l_03_capsule
   center=<5.773 -0.0242 4.9416>
   rotation=<rot 0.0003 -0.0003 87.2652>
  }]
 dyn_shirt_r_04 [{
   shapeType=Capsule
   length=13.0458
   radius=4.7798
   shapeName=dyn_shirt_l_04_capsule
   center=<5.773 0 4.8>
   rotation=<rot 0.0004 -0.0003 79.5129>
  }]
 dyn_shirt_r_MSTR [{shapeType=Capsule, length=0.505, radius=0.505, shapeName=dyn_shirt_r_MSTR_capsule, rotation=<rot 0 -90 0>}]
 pelvis [{shapeType=Capsule, length=5.3166, radius=14.1466, shapeName=pelvis_capsule, center=<-4.6738 -1.0033 0>}]
 thigh_l [{
   shapeType=Capsule
   length=36.317
   radius=9.9799
   shapeName=thigh_l_capsule
   center=<-14.2867 1.6302 0>
   rotation=<rot 86.2835 179.9998 179.9999>
  }]
 thigh_r [{
   shapeType=Capsule
   length=35.2559
   radius=9.9309
   shapeName=thigh_r_capsule
   center=<14.568 -1.9122 0>
   rotation=<rot 88.2493 179.9996 179.9997>
  }]
}
constraints [
 {bone1Name=dyn_shirt_l_01, bone2Name=dyn_shirt_l_MSTR, swing1LimitDegrees=1.5, swing2LimitDegrees=1.5, twistLimitDegrees=2}
 {bone1Name=dyn_shirt_l_02, bone2Name=dyn_shirt_l_01, swing1LimitDegrees=1.5, swing2LimitDegrees=1.5, twistLimitDegrees=2}
 {bone1Name=dyn_shirt_l_03, bone2Name=dyn_shirt_l_02, swing1LimitDegrees=1.5, swing2LimitDegrees=1.5, twistLimitDegrees=2}
 {bone1Name=dyn_shirt_l_04, bone2Name=dyn_shirt_l_03, swing1LimitDegrees=1.5, swing2LimitDegrees=1.5, twistLimitDegrees=2}
 {bone1Name=dyn_shirt_r_01, bone2Name=dyn_shirt_r_MSTR, swing1LimitDegrees=1.5, swing2LimitDegrees=1.5, twistLimitDegrees=2}
 {bone1Name=dyn_shirt_r_02, bone2Name=dyn_shirt_r_01, swing1LimitDegrees=1.5, swing2LimitDegrees=1.5, twistLimitDegrees=2}
 {bone1Name=dyn_shirt_r_03, bone2Name=dyn_shirt_r_02, swing1LimitDegrees=1.5, swing2LimitDegrees=1.5, twistLimitDegrees=2}
 {bone1Name=dyn_shirt_r_04, bone2Name=dyn_shirt_r_03, swing1LimitDegrees=1.5, swing2LimitDegrees=1.5, twistLimitDegrees=2}
 {bone1Name=dyn_drawstring_r_01, bone2Name=dyn_drawstring_l_MSTR, swing1LimitDegrees=20, swing2LimitDegrees=20, twistLimitDegrees=2}
 {bone1Name=dyn_drawstring_r_02, bone2Name=dyn_drawstring_r_01, swing1LimitDegrees=20, swing2LimitDegrees=20, twistLimitDegrees=2}
 {bone1Name=dyn_drawstring_r_03, bone2Name=dyn_drawstring_r_02, swing1LimitDegrees=20, swing2LimitDegrees=20, twistLimitDegrees=2}
 {bone1Name=dyn_drawstring_loop_r_01, bone2Name=dyn_drawstring_loop_l_MSTR, swing1LimitDegrees=4, swing2LimitDegrees=4, twistLimitDegrees=2}
 {bone1Name=dyn_drawstring_loop_r_02, bone2Name=dyn_drawstring_loop_r_01, swing1LimitDegrees=4, swing2LimitDegrees=4, twistLimitDegrees=2}
 {bone1Name=dyn_drawstring_loop_l_01, bone2Name=dyn_drawstring_loop_r_MSTR, swing1LimitDegrees=4, swing2LimitDegrees=4, twistLimitDegrees=2}
 {bone1Name=dyn_drawstring_loop_l_02, bone2Name=dyn_drawstring_loop_l_01, swing1LimitDegrees=4, swing2LimitDegrees=4, twistLimitDegrees=2}
 {bone1Name=dyn_drawstring_l_01, bone2Name=dyn_drawstring_r_MSTR, swing1LimitDegrees=20, swing2LimitDegrees=20, twistLimitDegrees=2}
 {bone1Name=dyn_drawstring_l_02, bone2Name=dyn_drawstring_l_01, swing1LimitDegrees=20, swing2LimitDegrees=20, twistLimitDegrees=2}
 {bone1Name=dyn_drawstring_l_03, bone2Name=dyn_drawstring_l_02, swing1LimitDegrees=20, swing2LimitDegrees=20, twistLimitDegrees=2}
]
