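# Physics asset for the M_MED_FloralMane_Sand male torso cosmetic, setting up a dynamic necklace over the player body.
# Capsules cover the pelvis, spine and necklace bones, with no constraints joining torso to necklace.
# Tightly limited, mostly swing- and twist-locked constraints chain the tag from its master bone and branch into upper, mid and lower bead strands.
# Extra angularly free constraints link tag bones back to bead bones, some duplicating existing pairs.
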PhysicsAsset /BRCosmetics/Characters/Player/Male/Medium/Bodies/M_MED_FloralMane_Sand/Meshes/M_MED_FloralMane_Sand_Torso_Physics
bodies {
 dyn_necklace_beads_lower_01 [{
   shapeType=Capsule
   length=2.3273
   radius=3.6727
   shapeName=dyn_necklace_beads_lower_01_capsule
   center=<2.7656 0 2.8374>
   rotation=<rot 0 0 90>
  }]
 dyn_necklace_beads_lower_02 [{
   shapeType=Capsule
   length=2.0391
   radius=3.9609
   shapeName=dyn_necklace_beads_lower_02_capsule
   center=<3.3272 0 2.8374>
   rotation=<rot 0 0 90>
  }]
 dyn_necklace_beads_mid_01 [{shapeType=Capsule, length=2.1708, radius=3.8292, shapeName=dyn_necklace_beads_mid_01_capsule, center=<2.4168 0 1.697>, rotation=<rot 0 0 90>}]
 dyn_necklace_beads_mid_02 [{
   shapeType=Capsule
   length=1.4974
   radius=4.5026
   shapeName=dyn_necklace_beads_mid_02_capsule
   center=<2.4168 0 3.7275>
   rotation=<rot 0 0 90>
  }]
 dyn_necklace_beads_upper_01 [{
   shapeType=Capsule
   length=10.8536
   radius=3.8225
   shapeName=dyn_necklace_beads_upper_01_capsule
   center=<3.3023 0 2.6483>
   rotation=<rot 0 0 90>
  }]
 dyn_necklace_beads_upper_02 [{
   shapeType=Capsule
   length=7.7207
   radius=4.6756
   shapeName=dyn_necklace_beads_upper_02_capsule
   center=<1.2417 0 3.6633>
   rotation=<rot 0 -179.6366 -89.9849>
  }]
 dyn_necklace_tag_01 [{shapeType=Capsule, length=11.4893, radius=4.1416, shapeName=dyn_necklace_tag_01_capsule, center=<7.7714 0 3.6552>, rotation=<rot 0 0 90>}]
 dyn_necklace_tag_02 [{shapeType=Capsule, length=10.5348, radius=3.8422, shapeName=dyn_necklace_tag_02_capsule, center=<6.253 0 3.005>, rotation=<rot 0 0 90>}]
 dyn_necklace_tag_03 [{shapeType=Capsule, length=8.3618, radius=3.6375, shapeName=dyn_necklace_tag_03_capsule, center=<5.9822 0 2.8561>, rotation=<rot 0 0 90>}]
 dyn_necklace_tag_04 [{
   shapeType=Capsule
   length=6.3132
   radius=5.0468
   shapeName=dyn_necklace_tag_04_capsule
   center=<7.694 0.024 4.5423>
   rotation=<rot 0 90.3862 -89.1138>
  }]
 dyn_necklace_tag_MSTR [{shapeType=Capsule, length=0.505, radius=0.505, shapeName=dyn_necklace_tag_MSTR_capsule, rotation=<rot 0 -90 0>}]
 dyn_necklaces_MSTR [{shapeType=Capsule, length=0.505, radius=0.505, shapeName=dyn_necklaces_MSTR_capsule, rotation=<rot 0 -90 0>}]
 pelvis [{
   shapeType=Capsule
   length=5.5216
   radius=12.8594
   shapeName=pelvis_capsule
   center=<1.5649 -1.0378 0.1213>
   rotation=<rot 3.4229 -90 -0.8716>
  }]
 spine_01 [{shapeType=Capsule, length=7.2176, radius=11.949, shapeName=spine_01_capsule, rotation=<rot 0 -90 0>}]
 spine_02 [{shapeType=Capsule, length=5.0938, radius=12.9006, shapeName=spine_02_capsule}]
 spine_03 [{shapeType=Capsule, length=3.158, radius=13.4173, shapeName=spine_03_capsule, center=<1.0068 -0.5056 0>}]
 spine_04 [{shapeType=Capsule, length=5.6204, radius=13.3638, shapeName=spine_04_capsule, center=<-0.7591 0 0>}]
 spine_05 [{shapeType=Capsule, length=7.8511, radius=13.3187, shapeName=spine_05_capsule, center=<-5.4501 -1.0262 0>}]
}
constraints [
 {
  bone1Name=dyn_necklace_tag_01
  bone2Name=dyn_necklace_tag_MSTR
  swing1LimitDegrees=1
  swing1Motion=Locked
  swing2LimitDegrees=1
  twistLimitDegrees=2
  twistMotion=Locked
 }
 {
  bone1Name=dyn_necklace_tag_02
  bone2Name=dyn_necklace_tag_01
  swing1LimitDegrees=1
  swing1Motion=Locked
  swing2LimitDegrees=1
  twistLimitDegrees=2
  twistMotion=Locked
 }
 {
  bone1Name=dyn_necklace_tag_03
  bone2Name=dyn_necklace_tag_02
  swing1LimitDegrees=1
  swing1Motion=Locked
  swing2LimitDegrees=1
  twistLimitDegrees=2
  twistMotion=Locked
 }
 {bone1Name=dyn_necklace_tag_04, bone2Name=dyn_necklace_tag_03, swing1LimitDegrees=2, swing2LimitDegrees=10, twistLimitDegrees=30}
 {
  bone1Name=dyn_necklace_beads_upper_01
  bone2Name=dyn_necklace_tag_01
  swing1LimitDegrees=0
  swing1Motion=Locked
  swing2LimitDegrees=1
  twistLimitDegrees=2
  twistMotion=Locked
 }
 {
  bone1Name=dyn_necklace_beads_upper_02
  bone2Name=dyn_necklace_beads_upper_01
  swing1LimitDegrees=1
  swing1Motion=Locked
  swing2LimitDegrees=1
  twistLimitDegrees=2
  twistMotion=Locked
 }
 {
  bone1Name=dyn_necklace_beads_mid_01
  bone2Name=dyn_necklace_tag_01
  swing1LimitDegrees=1
  swing1Motion=Locked
  swing2LimitDegrees=1
  twistLimitDegrees=45
  twistMotion=Locked
 }
 {
  bone1Name=dyn_necklace_beads_mid_02
  bone2Name=dyn_necklace_beads_mid_01
  swing1LimitDegrees=1
  swing1Motion=Locked
  swing2LimitDegrees=1
  twistLimitDegrees=45
  twistMotion=Locked
 }
 {
  bone1Name=dyn_necklace_beads_lower_01
  bone2Name=dyn_necklace_tag_03
  swing1LimitDegrees=1
  swing1Motion=Locked
  swing2LimitDegrees=1
  twistLimitDegrees=45
  twistMotion=Locked
 }
 {
  bone1Name=dyn_necklace_beads_lower_02
  bone2Name=dyn_necklace_beads_lower_01
  swing1LimitDegrees=1
  swing1Motion=Locked
  swing2LimitDegrees=1.5
  twistLimitDegrees=2
 }
 {
  bone1Name=dyn_necklace_tag_01
  bone2Name=dyn_necklace_beads_mid_01
  swing1LimitDegrees=45
  swing1Motion=Free
  swing2LimitDegrees=45
  swing2Motion=Free
  twistLimitDegrees=45
  twistMotion=Free
 }
 {
  bone1Name=dyn_necklace_tag_01
  bone2Name=dyn_necklace_beads_upper_02
  swing1LimitDegrees=45
  swing1Motion=Free
  swing2LimitDegrees=45
  swing2Motion=Free
  twistLimitDegrees=45
  twistMotion=Free
 }
 {
  bone1Name=dyn_necklace_tag_01
  bone2Name=dyn_necklace_beads_upper_01
  swing1LimitDegrees=45
  swing1Motion=Free
  swing2LimitDegrees=45
  swing2Motion=Free
  twistLimitDegrees=45
  twistMotion=Free
 }
 {
  bone1Name=dyn_necklace_tag_03
  bone2Name=dyn_necklace_beads_lower_02
  swing1LimitDegrees=45
  swing1Motion=Free
  swing2LimitDegrees=45
  swing2Motion=Free
  twistLimitDegrees=45
  twistMotion=Free
 }
]
