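# Post-process style AnimBP for the M_MED_FloralMane_Sand player body cosmetic, adding secondary motion to the pose copied from the parent mesh.
# A Control Rig feeds torso and pelvis RigidBody simulations in base-bone space, then spring-driven AnimDynamics for shorts and clog-loop bones.
# The event graph stores gravity and rigid-body settings from a parameter-asset function each update.
# It zeroes AnimDynamicsAlpha during skydiving, grind rails and ziplines, restoring it afterward.
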AnimBlueprint /BRCosmetics/Characters/Player/Male/Medium/Bodies/M_MED_FloralMane_Sand/Meshes/M_MED_FloralMane_Sand_AnimBP
target_skeleton=&/M_MED_FloralMane_Sand_Skeleton
graphs [
 AnimGraph {nodes [
   n1 = Control Rig(Source=Copy Pose From Mesh() {bUseAttachedParent=true}) {controlRigAssetReference {blueprintRigClass=&/M_MED_FloralMane_Sand_CtrlRig.M_MED_FloralMane_Sand_CtrlRig_C}}
   n2 = RigidBody!(ComponentPose=Local To Component(LocalPose=@n1), Alpha, OverrideWorldGravity:Vector="0,0,-980", ExternalForce) -> (Pose) {
    bOverrideWorldGravity=true
    overridePhysicsAsset=&/M_MED_FloralMane_Sand_Torso_Physics
    simulationSpace=BaseBoneSpace
    baseBoneRef {boneName=spine_05}
    overrideWorldGravity {z=-980}
    simSpaceSettings {
     dampingAlpha=0
     maxAngularAcceleration=20
     maxAngularVelocity=4
     maxLinearAcceleration=1000
     maxLinearVelocity=200
     worldAlpha=1
     externalLinearDragV=<0.05 0.05 0.05>
    }
   }
   n3 = RigidBody!(
    ComponentPose=@n2
    Alpha
    OverrideWorldGravity="0,0,-980"
    ExternalForce
    SimSpaceSettings:Sim Space Settings Structure=(WorldAlpha=1,VelocityScaleZ=0.6,DampingAlpha=0,MaxLinearVelocity=600,MaxAngularVelocity=8,MaxLinearAcceleration=3000,MaxAngularAcceleration=40,ExternalLinearDragV=(X=0.1,Y=0.1,Z=0.1),ExternalLinearVelocity=(X=0,Y=0,Z=0),ExternalAngularVelocity=(X=0,Y=0,Z=0))
   ) -> (Pose) {
    bOverrideWorldGravity=true
    overridePhysicsAsset=&/M_MED_FloralMane_Sand_Pelvis_Physics
    simulationSpace=BaseBoneSpace
    baseBoneRef {boneName=pelvis}
    overrideWorldGravity {z=-980}
    simSpaceSettings {
     dampingAlpha=0
     maxAngularAcceleration=40
     maxAngularVelocity=8
     maxLinearAcceleration=3000
     maxLinearVelocity=600
     velocityScaleZ=0.6
     worldAlpha=1
     externalLinearDragV=<0.1 0.1 0.1>
    }
   }
   n4 = Anim Dynamics - Bone: dyn_shorts_l_01!(ComponentPose=@n3, Alpha=1, GravityOverride:Vector="0,0,-980", ExternalForce="0,0,0") -> (Pose) {
    angularDampingOverride=0.999
    angularSpringConstant=15000
    bAngularSpring=true
    bOverrideAngularDamping=true
    bUseGravityOverride=true
    simulationSpace=RootRelative
    alphaScaleBiasClamp {bInterpResult=true}
    boundBone {boneName=dyn_shorts_l_01}
    componentAppliedLinearAccClamp=<3000 3000 3000>
    componentLinearAccScale=<0.2 0.2 0.2>
    componentLinearVelScale=<0.2 0.2 0.2>
    gravityOverride {z=-980}
    physicsBodyDefinitions [{
      boundBone {boneName=dyn_shorts_l_01}
      boxExtents=<45 45 45>
      constraintSetup {angularLimitsMax=<0 2 2>, angularLimitsMin=<0 -2 -2>, angularTarget=<1 0 0>}
      localJointOffset=<0.25 0 0>
     }]
    simSpaceSettings {maxAngularAcceleration=40, maxAngularVelocity=8, simSpaceAngularAlpha=0.4}
   }
   n5 = Anim Dynamics - Bone: dyn_shorts_l_02!(ComponentPose=@n4, Alpha=1, GravityOverride:Vector="0,0,-980", ExternalForce="0,0,0") -> (Pose) {
    angularDampingOverride=0.999
    angularSpringConstant=20000
    bAngularSpring=true
    bOverrideAngularDamping=true
    bUseGravityOverride=true
    simulationSpace=RootRelative
    alphaScaleBiasClamp {bInterpResult=true}
    boundBone {boneName=dyn_shorts_l_02}
    componentAppliedLinearAccClamp=<3000 3000 3000>
    componentLinearAccScale=<0.2 0.2 0.2>
    componentLinearVelScale=<0.2 0.2 0.2>
    gravityOverride {z=-980}
    physicsBodyDefinitions [{
      boundBone {boneName=dyn_shorts_l_02}
      boxExtents=<40 40 40>
      constraintSetup {angularLimitsMax=<0 2 2>, angularLimitsMin=<0 -2 -2>, angularTarget=<1 0 0>}
      localJointOffset=<0.25 0 0>
     }]
    simSpaceSettings {maxAngularAcceleration=60, maxAngularVelocity=12, simSpaceAngularAlpha=0.75}
   }
   n6 = Anim Dynamics - Bone: dyn_shorts_r_01!(ComponentPose=@n5, Alpha=1, GravityOverride:Vector="0,0,-980", ExternalForce="0,0,0") -> (Pose) {
    angularDampingOverride=0.999
    angularSpringConstant=20000
    bAngularSpring=true
    bOverrideAngularDamping=true
    bUseGravityOverride=true
    simulationSpace=RootRelative
    alphaScaleBiasClamp {bInterpResult=true}
    boundBone {boneName=dyn_shorts_r_01}
    componentAppliedLinearAccClamp=<3000 3000 3000>
    componentLinearAccScale=<0.2 0.2 0.2>
    componentLinearVelScale=<0.2 0.2 0.2>
    gravityOverride {z=-980}
    physicsBodyDefinitions [{
      boundBone {boneName=dyn_shorts_r_01}
      boxExtents=<40 40 40>
      constraintSetup {angularLimitsMax=<0 2 2>, angularLimitsMin=<0 -2 -2>, angularTarget=<1 0 0>}
      localJointOffset=<0.25 0 0>
     }]
    simSpaceSettings {maxAngularAcceleration=60, maxAngularVelocity=12, simSpaceAngularAlpha=0.75}
   }
   n7 = Anim Dynamics - Bone: dyn_clog_loop_l!(ComponentPose=@n6, Alpha=1, GravityOverride:Vector="0,0,-980", ExternalForce="0,0,0") -> (Pose) {
    angularSpringConstant=5000
    bAngularSpring=true
    bUseGravityOverride=true
    simulationSpace=RootRelative
    alphaScaleBiasClamp {bInterpResult=true}
    boundBone {boneName=dyn_clog_loop_l}
    componentAppliedLinearAccClamp=<3000 3000 3000>
    componentLinearAccScale=<0.3 0.3 0.3>
    componentLinearVelScale=<0.3 0.3 0.3>
    gravityOverride {z=-980}
    physicsBodyDefinitions [{
      boundBone {boneName=dyn_clog_loop_l}
      boxExtents=<20 20 20>
      constraintSetup {angularLimitsMax=<0 20 1>, angularLimitsMin=<0 -20 -1>, angularTarget=<1 0 0>}
      localJointOffset=<1 0 0>
     }]
    relativeSpaceBone {boneName=foot_l}
    simSpaceSettings {maxAngularAcceleration=60, maxAngularVelocity=12}
   }
   n8 = Anim Dynamics - Bone: dyn_clog_loop_r!(ComponentPose=@n7, Alpha=1, GravityOverride:Vector="0,0,-980", ExternalForce="0,0,0") -> (Pose) {
    angularSpringConstant=5000
    bAngularSpring=true
    bUseGravityOverride=true
    simulationSpace=RootRelative
    alphaScaleBiasClamp {bInterpResult=true}
    boundBone {boneName=dyn_clog_loop_r}
    componentAppliedLinearAccClamp=<3000 3000 3000>
    componentLinearAccScale=<0.3 0.3 0.3>
    componentLinearVelScale=<0.3 0.3 0.3>
    gravityOverride {z=-980}
    physicsBodyDefinitions [{
      boundBone {boneName=dyn_clog_loop_r}
      boxExtents=<20 20 20>
      constraintSetup {angularLimitsMax=<0 20 1>, angularLimitsMin=<0 -20 -1>, angularTarget=<1 0 0>}
      localJointOffset=<1 0 0>
     }]
    relativeSpaceBone {boneName=foot_l}
    simSpaceSettings {maxAngularAcceleration=60, maxAngularVelocity=12}
   }
   Output Pose(Result=Component To Local(ComponentPose=@n8)) {layerGroup=None}
  ]}
 EventGraph {nodes [
   Event OnBeginSkydiving()
   > Set AnimDynamicsAlpha(AnimDynamicsAlpha:double=0) -> (Output_Get:double)
   Event OnEndSkydiving()
   > Set AnimDynamicsAlpha(AnimDynamicsAlpha=1)
   Event BlueprintUpdateAnimation()
   > n1 = "Gravity Override State Function (Parameter Asset) (Nativized)"(
    CharacterPartAnimInstance=Self-Reference()
    GameTimeInSeconds=GetGameTimeInSeconds()
    Parameters=/BRCosmetics/Characters/Player/Male/Medium/Bodies/M_MED_FloralMane_Sand/Meshes/Parts/M_MED_FloralMane_Sand_FA_Parameters
   )
   > Set Gravity Override(Gravity Override=@n1)
   > Set RBAN Settings Out(RBAN Settings Out=@n1.OutRigidBodyAnimNodeParameters)
   Event OnEnterGrindRail() {is=title K2Node_Event}
   > Set AnimDynamicsAlpha(AnimDynamicsAlpha=0)
   n2 = Event BlueprintProcessAnimInputEvent()
   > n3 = Branch(Condition=IsNamed(self=@n2.AnimInputEvent, InEventName=aie.Player.Zipline.Entered))
   > Set AnimDynamicsAlpha(AnimDynamicsAlpha=0)
   Branch(execute=@n3.else, Condition=IsNamed(self=@n2.AnimInputEvent, InEventName=aie.Player.Zipline.Exited))
   > Set AnimDynamicsAlpha(AnimDynamicsAlpha=1)
   Event OnExitGrindRail() {is=title K2Node_Event}
   > Set AnimDynamicsAlpha(AnimDynamicsAlpha=1)
  ]}
]
variables [
 AnimDynamicsAlpha {type=number, default=1}
 Gravity Override {type=Vector}
 RBAN Settings Out {type=FortRigidBodyAnimNodeParameters}
]
