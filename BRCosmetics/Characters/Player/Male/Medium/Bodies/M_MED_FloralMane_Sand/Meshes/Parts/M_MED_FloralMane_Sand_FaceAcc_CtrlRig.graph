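# Control Rig for the FloralMane Sand face-accessory mesh, posing its dynamic hair bones against head, spine and arm motion.
# By default, aim constraints, a SkydiveAlpha pose and spherical pose readers on head and upper arms drive corrective offsets.
# With useSplineHair, weighted parent-constrained nulls carry controls for spline IK on three back chains, initialised from bones at construction.
# Both branches output arm-raise values to arm_l/r_up_SPR curves.
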ControlRigBlueprint /BRCosmetics/Characters/Player/Male/Medium/Bodies/M_MED_FloralMane_Sand/Meshes/Parts/M_MED_FloralMane_Sand_FaceAcc_CtrlRig
graphs [
 RigVMFunctionLibrary {links [], nodes []}
 RigVMFunction_Sequence_1_ContainedGraph {nodes [
   n1 = RigVMFunctionEntryNode()
   RigVMFunction_Sequence = &RigVMFunction_Sequence(ExecuteContext=@n1.ExecuteContext)
   n2 = &RigVMFunction_Sequence(ExecuteContext=@RigVMFunction_Sequence.B)
   RigVMFunctionReturnNode(A=@RigVMFunction_Sequence.A, B=@n2.A, C=@n2.B)
  ]}
 RigVMModel {nodes [
   &RigUnit_AimConstraintLocalSpaceOffset(
    Child=(Type=Bone,Name="dyn_hair_back_l_MSTR")
    ExecutePin=()
    Filter=(bX=True,bY=True,bZ=True)
    Parents=((Item=(Type=Bone,Name="spine_04"),Weight=1))
    Weight=1
    WorldUp=(Target=(X=0,Y=0,Z=1),Kind=Location,Space=(Type=Bone,Name="spine_04"))
   )
   > &RigUnit_AimConstraintLocalSpaceOffset(
    Child=(Type=Bone,Name="dyn_hair_back_r_MSTR")
    Filter=(bX=True,bY=True,bZ=True)
    Parents=((Item=(Type=Bone,Name="spine_04"),Weight=1))
    Weight=1
    WorldUp=(Target=(X=0,Y=0,Z=1),Kind=Location,Space=(Type=Bone,Name="spine_04"))
   )
   RigUnit_BeginExecution = &RigUnit_BeginExecution()
   RigVMFunction_ControlFlowBranch = &RigVMFunction_ControlFlowBranch(ExecuteContext=@RigUnit_BeginExecution.ExecutePin, Condition=RigVMVariableNode() {variable=useSplineHair}.Value)
   RigVMFunction_Sequence = RigVMAggregateNode(ExecuteContext=@RigVMFunction_ControlFlowBranch.False)
   &RigUnit_AimConstraintLocalSpaceOffset(
    Child=(Type=Bone,Name="dyn_hair_back_mid_MSTR")
    Filter=(bX=True,bY=True,bZ=True)
    Parents=((Item=(Type=Bone,Name="spine_05"),Weight=1))
    Weight=1
    WorldUp=(Target=(X=0,Y=1,Z=0),Kind=Location,Space=(Type=Bone,Name="spine_05"))
    ExecutePin=@RigVMFunction_Sequence.A
   )
   > &RigUnit_AimConstraintLocalSpaceOffset(
    Child=(Type=Bone,Name="dyn_hair_front_l_MSTR")
    Filter=(bX=True,bY=True,bZ=True)
    Parents=((Item=(Type=Bone,Name="spine_05"),Weight=1))
    Weight=1
    WorldUp=(Target=(X=0,Y=1,Z=0),Kind=Location,Space=(Type=Bone,Name="spine_05"))
   )
   > &RigUnit_AimConstraintLocalSpaceOffset(
    Child=(Type=Bone,Name="dyn_hair_front_r_MSTR")
    Filter=(bX=True,bY=True,bZ=True)
    Parents=((Item=(Type=Bone,Name="spine_05"),Weight=1))
    Weight=1
    WorldUp=(Target=(X=0,Y=1,Z=0),Kind=Location,Space=(Type=Bone,Name="spine_05"))
   )
   > &RigUnit_ModifyTransforms(
    ItemToModify="((Item=(Type=Bone,Name=\"dyn_hair_back_l_MSTR\"),Transform=(Rotation=(X=0.0464,Y=-0.0737,Z=-0.2599,W=0.9617),Translation=(X=-0.0001,Y=-5.1052,Z=3.9865),Scale3D=(X=1,Y=1,Z=1))),(Item=(Type=Bone,Name=\"dyn_hair_back_mid_MSTR\"),Transform=(Rotation=(X=0,Y=-0.0671 ...(truncated)"
    WeightMinimum=0
    Weight=RigVMVariableNode() {variable=SkydiveAlpha}.Value
   )
   n1 = &RigUnit_SphericalPoseReader(
    ActiveRegionScaleFactors=(PositiveWidth=1,NegativeWidth=1,PositiveHeight=1,NegativeHeight=1)
    ActiveRegionSize=0.1
    Debug=(bDrawDebug=false,bDraw2D=False,bDrawLocalAxes=False,DebugScale=25,DebugSegments=20,DebugThickness=0.25)
    DriverAxis=(X=0,Y=1,Z=0)
    DriverItem=(Type=Bone,Name="head")
    FalloffRegionScaleFactors=(PositiveWidth=2,NegativeWidth=2,PositiveHeight=1,NegativeHeight=1)
    FalloffSize=0.2
    FlipHeightScaling=False
    FlipWidthScaling=False
    OptionalParentItem=(Type=Bone,Name="spine_05")
    RotationOffset=(X=-45,Y=-90,Z=90)
    ExecutePin=@RigVMFunction_Sequence.B
   )
   n2 = &RigVMFunction_AlphaInterp(InterpSpeedDecreasing=10, InterpSpeedIncreasing=10, Scale=1, bClampResult=False, bInterpResult=True, Value=@n1.OutputParam)
   > &RigUnit_ModifyTransforms(
    ItemToModify="((Item=(Type=Bone,Name=\"dyn_hair_back_mid_MSTR\"),Transform=(Rotation=(X=0,Y=-0.1736,Z=0,W=0.9848),Translation=(X=0,Y=0,Z=0),Scale3D=(X=1,Y=1,Z=1))),(Item=(Type=Bone,Name=\"dyn_hair_back_r_MSTR\"),Transform=(Rotation=(X=0.0302,Y=-0.171,Z=0.171,W=0.9698),Trans ...(truncated)"
    WeightMinimum=0
    Weight=@n2.Result
   )
   > n3 = &RigUnit_SphericalPoseReader(
    ActiveRegionScaleFactors=(PositiveWidth=1,NegativeWidth=1,PositiveHeight=1,NegativeHeight=1)
    ActiveRegionSize=0.1
    Debug=(bDrawDebug=false,bDraw2D=False,bDrawLocalAxes=False,DebugScale=25,DebugSegments=20,DebugThickness=0.25)
    DriverAxis=(X=0,Y=1,Z=0)
    DriverItem=(Type=Bone,Name="head")
    FalloffRegionScaleFactors=(PositiveWidth=2,NegativeWidth=2,PositiveHeight=1,NegativeHeight=1)
    FalloffSize=0.2
    FlipHeightScaling=False
    FlipWidthScaling=False
    OptionalParentItem=(Type=Bone,Name="spine_05")
    RotationOffset=(X=-90,Y=-90,Z=90)
   )
   RigVMFunction_AlphaInterp = &RigVMFunction_AlphaInterp(InterpSpeedDecreasing=20, InterpSpeedIncreasing=20, Scale=1, bClampResult=False, bInterpResult=True, Value=@n3.OutputParam)
   > &RigUnit_ModifyTransforms(
    ItemToModify="((Item=(Type=Bone,Name=\"dyn_hair_back_mid_MSTR\"),Transform=(Rotation=(X=0.0078,Y=-0.5332,Z=-0.0847,W=0.8417),Translation=(X=0,Y=0,Z=2),Scale3D=(X=1,Y=1,Z=1))),(Item=(Type=Bone,Name=\"dyn_hair_back_l_MSTR\"),Transform=(Rotation=(X=0.1814,Y=-0.0486,Z=-0.2542,W ...(truncated)"
    WeightMinimum=0
    Weight=@RigVMFunction_AlphaInterp.Result
   )
   > n4 = &RigUnit_SphericalPoseReader(
    ActiveRegionScaleFactors=(PositiveWidth=1.5,NegativeWidth=1.5,PositiveHeight=2,NegativeHeight=1)
    ActiveRegionSize=0.1
    Debug=(bDrawDebug=false,bDraw2D=False,bDrawLocalAxes=False,DebugScale=25,DebugSegments=20,DebugThickness=0.25)
    DriverAxis=(X=0,Y=1,Z=0)
    DriverItem=(Type=Bone,Name="head")
    FalloffRegionScaleFactors=(PositiveWidth=1,NegativeWidth=1,PositiveHeight=1.5,NegativeHeight=1)
    FalloffSize=0.4
    FlipHeightScaling=False
    FlipWidthScaling=False
    OptionalParentItem=(Type=Bone,Name="None")
    RotationOffset=(X=200,Y=0,Z=180)
   )
   > &RigUnit_ModifyTransforms(
    ItemToModify="((Item=(Type=Bone,Name=\"dyn_hair_back_mid_MSTR\"),Transform=(Rotation=(X=-0.0129,Y=0.0912,Z=0.0547,W=-0.9942),Translation=(X=0,Y=0,Z=0),Scale3D=(X=1,Y=1,Z=1))),(Item=(Type=Bone,Name=\"dyn_hair_back_l_MSTR\"),Transform=(Rotation=(X=0.1037,Y=0.0023,Z=-0.1707,W= ...(truncated)"
    WeightMinimum=0
    Weight=@n4.OutputParam
   )
   > n5 = &RigUnit_SphericalPoseReader(
    ActiveRegionScaleFactors=(PositiveWidth=1.5,NegativeWidth=1.5,PositiveHeight=2,NegativeHeight=1)
    ActiveRegionSize=0.1
    Debug=(bDrawDebug=false,bDraw2D=False,bDrawLocalAxes=False,DebugScale=25,DebugSegments=20,DebugThickness=0.25)
    DriverAxis=(X=0,Y=1,Z=0)
    DriverItem=(Type=Bone,Name="head")
    FalloffRegionScaleFactors=(PositiveWidth=1,NegativeWidth=1,PositiveHeight=1.5,NegativeHeight=1)
    FalloffSize=0.4
    FlipHeightScaling=False
    FlipWidthScaling=False
    OptionalParentItem=(Type=Bone,Name="None")
    RotationOffset=(X=40,Y=0,Z=0)
   )
   > &RigUnit_ModifyTransforms(
    ItemToModify="((Item=(Type=Bone,Name=\"dyn_hair_back_mid_MSTR\"),Transform=(Rotation=(X=-0.0078,Y=-0.0807,Z=0.0613,W=0.9948),Translation=(X=0,Y=0,Z=0),Scale3D=(X=1,Y=1,Z=1))),(Item=(Type=Bone,Name=\"dyn_hair_back_r_MSTR\"),Transform=(Rotation=(X=0.0611,Y=0.0419,Z=-0.0461,W= ...(truncated)"
    WeightMinimum=0
    Weight=@n5.OutputParam
   )
   > n6 = &RigUnit_SphericalPoseReader(
    ActiveRegionScaleFactors=(PositiveWidth=2,NegativeWidth=2,PositiveHeight=1,NegativeHeight=1)
    ActiveRegionSize=0.15
    Debug=(bDrawDebug=false,bDraw2D=False,bDrawLocalAxes=False,DebugScale=25,DebugSegments=20,DebugThickness=0.25)
    DriverAxis=(X=-1,Y=0,Z=0)
    DriverItem=(Type=Bone,Name="upperarm_r")
    FalloffRegionScaleFactors=(PositiveWidth=2,NegativeWidth=1.5,PositiveHeight=1.5,NegativeHeight=1)
    FalloffSize=0.4
    FlipHeightScaling=False
    FlipWidthScaling=False
    OptionalParentItem=(Type=Bone,Name="None")
    RotationOffset=(X=-119.3316,Y=0,Z=19.1823)
   )
   > &RigUnit_ModifyTransforms(
    ItemToModify="((Item=(Type=Bone,Name=\"dyn_hair_back_r_MSTR\"),Transform=(Rotation=(X=-0.0551,Y=0.302,Z=-0.0404,W=0.9509),Translation=(X=-0.3372,Y=5.612,Z=5.0799),Scale3D=(X=1,Y=1,Z=1))),(Item=(Type=Bone,Name=\"dyn_hair_front_r_MSTR\"),Transform=(Rotation=(X=0,Y=-0.181,Z=0, ...(truncated)"
    WeightMinimum=0
    Weight=@n6.OutputParam
   )
   > /Script/ControlRig.RigUnit_SetCurveValue(Curve=arm_r_up_SPR, Value=@n6.OutputParam) {is=struct RigVMUnitNode}
   > n7 = &RigUnit_SphericalPoseReader(
    ActiveRegionScaleFactors=(PositiveWidth=1,NegativeWidth=1,PositiveHeight=1,NegativeHeight=1)
    ActiveRegionSize=0.12
    Debug=(bDrawDebug=false,bDraw2D=False,bDrawLocalAxes=False,DebugScale=25,DebugSegments=20,DebugThickness=0.25)
    DriverAxis=(X=1,Y=0,Z=0)
    DriverItem=(Type=Bone,Name="upperarm_l")
    FalloffRegionScaleFactors=(PositiveWidth=2,NegativeWidth=2,PositiveHeight=1,NegativeHeight=1)
    FalloffSize=0.4
    FlipHeightScaling=False
    FlipWidthScaling=False
    OptionalParentItem=(Type=Bone,Name="None")
    RotationOffset=(X=50,Y=20,Z=-25)
   )
   > &RigUnit_SetCurveValue(Curve=arm_l_up_SPR, Value=@n7.OutputParam)
   > n8 = &RigUnit_SphericalPoseReader(
    ActiveRegionScaleFactors=(PositiveWidth=2,NegativeWidth=2,PositiveHeight=1,NegativeHeight=1)
    ActiveRegionSize=0.05
    Debug=(bDrawDebug=True,bDraw2D=False,bDrawLocalAxes=False,DebugScale=25,DebugSegments=20,DebugThickness=0.25)
    DriverAxis=(X=0,Y=1,Z=0)
    DriverItem=(Type=Bone,Name="head")
    FalloffRegionScaleFactors=(PositiveWidth=1,NegativeWidth=1,PositiveHeight=1,NegativeHeight=1)
    FalloffSize=0.175
    FlipHeightScaling=False
    FlipWidthScaling=False
    OptionalParentItem=(Type=Bone,Name="spine_05")
    RotationOffset=(X=135,Y=0,Z=25)
   )
   > &RigUnit_ModifyTransforms(
    ItemToModify=((Item=(Type=Bone,Name="dyn_hair_front_l_MSTR"),Transform=(Rotation=(X=-0.0248,Y=-0.1509,Z=-0.1605,W=0.9751),Translation=(X=0,Y=0,Z=0),Scale3D=(X=1,Y=1,Z=1))))
    WeightMinimum=0
    Weight=@n8.OutputParam
   )
   &RigUnit_AimConstraintLocalSpaceOffset(
    Child=(Type=Bone,Name="dyn_hair_back_l_MSTR")
    Filter=(bX=True,bY=True,bZ=True)
    Parents=((Item=(Type=Bone,Name="spine_04"),Weight=1))
    Weight=1
    WorldUp=(Target=(X=0,Y=0,Z=1),Kind=Location,Space=(Type=Bone,Name="spine_04"))
    ExecutePin=@RigVMFunction_ControlFlowBranch.True
   )
   > &RigUnit_AimConstraintLocalSpaceOffset(
    Child=(Type=Bone,Name="dyn_hair_back_mid_MSTR")
    Filter=(bX=True,bY=True,bZ=True)
    Parents=((Item=(Type=Bone,Name="spine_04"),Weight=1))
    Weight=1
    WorldUp=(Target=(X=0,Y=0,Z=1),Kind=Location,Space=(Type=Bone,Name="spine_04"))
   )
   > &RigUnit_AimConstraintLocalSpaceOffset(
    Child=(Type=Bone,Name="dyn_hair_back_r_MSTR")
    Filter=(bX=True,bY=True,bZ=True)
    Parents=((Item=(Type=Bone,Name="spine_04"),Weight=1))
    Weight=1
    WorldUp=(Target=(X=0,Y=0,Z=1),Kind=Location,Space=(Type=Bone,Name="spine_04"))
   )
   > &RigUnit_AimConstraintLocalSpaceOffset(
    Child=(Type=Bone,Name="dyn_hair_front_l_MSTR")
    Filter=(bX=True,bY=True,bZ=True)
    Parents=((Item=(Type=Bone,Name="spine_05"),Weight=1))
    Weight=1
    WorldUp=(Target=(X=0,Y=1,Z=0),Kind=Location,Space=(Type=Bone,Name="spine_05"))
   )
   > &RigUnit_AimConstraintLocalSpaceOffset(
    Child=(Type=Bone,Name="dyn_hair_front_r_MSTR")
    Filter=(bX=True,bY=True,bZ=True)
    Parents=((Item=(Type=Bone,Name="spine_05"),Weight=1))
    Weight=1
    WorldUp=(Target=(X=0,Y=1,Z=0),Kind=Location,Space=(Type=Bone,Name="spine_05"))
   )
   > &RigUnit_ParentConstraint(
    AdvancedSettings=(RotationOrderForFilter=XZY)
    Child=(Type=Null,Name="dyn_hair_back_mid_01_null")
    Parents="((Item=(Type=Bone,Name=\"upperarm_twist_01_r\"),Weight=0.0002),(Item=(Type=Bone,Name=\"deform_upperarm_up_r\"),Weight=0.0003),(Item=(Type=Bone,Name=\"spine_04\"),Weight=0.0011),(Item=(Type=Bone,Name=\"clavicle_l\"),Weight=0.0047),(Item=(Type=Bone,Name=\"clavicle_r\" ...(truncated)"
    Weight=1
   )
   > &RigUnit_ParentConstraint(
    AdvancedSettings=(RotationOrderForFilter=XZY)
    Child=(Type=Null,Name="dyn_hair_back_mid_02_null")
    Parents="((Item=(Type=Bone,Name=\"upperarm_twist_01_r\"),Weight=0.0002),(Item=(Type=Bone,Name=\"deform_upperarm_up_r\"),Weight=0.0003),(Item=(Type=Bone,Name=\"spine_04\"),Weight=0.0011),(Item=(Type=Bone,Name=\"clavicle_l\"),Weight=0.0047),(Item=(Type=Bone,Name=\"clavicle_r\" ...(truncated)"
    Weight=1
   )
   > &RigUnit_ParentConstraint(
    AdvancedSettings=(RotationOrderForFilter=XZY)
    Child=(Type=Null,Name="dyn_hair_back_mid_03_null")
    Parents="((Item=(Type=Bone,Name=\"spine_02\"),Weight=0.0001),(Item=(Type=Bone,Name=\"deform_upperarm_bk_r\"),Weight=0.0003),(Item=(Type=Bone,Name=\"deform_upperarm_bk_l\"),Weight=0.0003),(Item=(Type=Bone,Name=\"deform_upperarm_up_l\"),Weight=0.0008),(Item=(Type=Bone,Name=\" ...(truncated)"
    Weight=1
   )
   > &RigUnit_ParentConstraint(
    AdvancedSettings=(RotationOrderForFilter=XZY)
    Child=(Type=Null,Name="dyn_hair_back_r_01_null")
    Parents="((Item=(Type=Bone,Name=\"deform_upperarm_fr_r\"),Weight=0.0001),(Item=(Type=Bone,Name=\"spine_04\"),Weight=0.0009),(Item=(Type=Bone,Name=\"upperarm_twist_01_r\"),Weight=0.0018),(Item=(Type=Bone,Name=\"deform_upperarm_up_r\"),Weight=0.0029),(Item=(Type=Bone,Name=\"s ...(truncated)"
    Weight=1
   )
   > &RigUnit_ParentConstraint(
    AdvancedSettings=(RotationOrderForFilter=XZY)
    Child=(Type=Null,Name="dyn_hair_back_r_02_null")
    Parents="((Item=(Type=Bone,Name=\"deform_upperarm_fr_r\"),Weight=0.0001),(Item=(Type=Bone,Name=\"spine_04\"),Weight=0.0009),(Item=(Type=Bone,Name=\"upperarm_twist_01_r\"),Weight=0.0018),(Item=(Type=Bone,Name=\"deform_upperarm_up_r\"),Weight=0.0029),(Item=(Type=Bone,Name=\"s ...(truncated)"
    Weight=1
   )
   > &RigUnit_ParentConstraint(
    AdvancedSettings=(RotationOrderForFilter=XZY)
    Child=(Type=Null,Name="dyn_hair_back_r_03_null")
    Parents="((Item=(Type=Bone,Name=\"upperarm_twist_02_r\"),Weight=0.0001),(Item=(Type=Bone,Name=\"deform_upperarm_bk_r\"),Weight=0.0003),(Item=(Type=Bone,Name=\"deform_upperarm_fr_r\"),Weight=0.0011),(Item=(Type=Bone,Name=\"spine_04\"),Weight=0.0027),(Item=(Type=Bone,Name=\"u ...(truncated)"
    Weight=1
   )
   > &RigUnit_ParentConstraint(
    AdvancedSettings=(RotationOrderForFilter=XZY)
    Child=(Type=Null,Name="dyn_hair_back_l_01_null")
    Parents="((Item=(Type=Bone,Name=\"upperarm_twist_01_l\"),Weight=0.0004),(Item=(Type=Bone,Name=\"spine_04\"),Weight=0.0004),(Item=(Type=Bone,Name=\"deform_upperarm_up_l\"),Weight=0.0005),(Item=(Type=Bone,Name=\"clavicle_r\"),Weight=0.0008),(Item=(Type=Bone,Name=\"spine_03\"), ...(truncated)"
    Weight=1
   )
   > &RigUnit_ParentConstraint(
    AdvancedSettings=(RotationOrderForFilter=XZY)
    Child=(Type=Null,Name="dyn_hair_back_l_02_null")
    Parents="((Item=(Type=Bone,Name=\"clavicle_r\"),Weight=0.0001),(Item=(Type=Bone,Name=\"spine_04\"),Weight=0.0004),(Item=(Type=Bone,Name=\"upperarm_twist_01_l\"),Weight=0.001),(Item=(Type=Bone,Name=\"deform_upperarm_up_l\"),Weight=0.0017),(Item=(Type=Bone,Name=\"spine_03\"),W ...(truncated)"
    Weight=1
   )
   > &RigUnit_ParentConstraint(
    AdvancedSettings=(RotationOrderForFilter=XZY)
    Child=(Type=Null,Name="dyn_hair_back_l_03_null")
    Parents="((Item=(Type=Bone,Name=\"clavicle_r\"),Weight=0.0002),(Item=(Type=Bone,Name=\"deform_upperarm_bk_l\"),Weight=0.0003),(Item=(Type=Bone,Name=\"upperarm_twist_02_l\"),Weight=0.0004),(Item=(Type=Bone,Name=\"deform_upperarm_fr_l\"),Weight=0.0014),(Item=(Type=Bone,Name= ...(truncated)"
    Weight=1
   )
   > &RigUnit_ModifyTransforms(
    ItemToModify="((Item=(Type=Null,Name=\"dyn_hair_back_l_02_null\"),Transform=(Rotation=(X=0,Y=0,Z=0,W=1),Translation=(X=0,Y=0,Z=2),Scale3D=(X=1,Y=1,Z=1))),(Item=(Type=Null,Name=\"dyn_hair_back_mid_02_null\"),Transform=(Rotation=(X=0,Y=0,Z=0,W=1),Translation=(X=0,Y=0,Z=2),Sca ...(truncated)"
    Weight=1
    WeightMinimum=0
   )
   > n9 = &RigUnit_SphericalPoseReader(
    ActiveRegionScaleFactors=(PositiveWidth=1,NegativeWidth=1,PositiveHeight=1,NegativeHeight=1)
    ActiveRegionSize=0.1
    Debug=(bDrawDebug=false,bDraw2D=False,bDrawLocalAxes=False,DebugScale=25,DebugSegments=20,DebugThickness=0.25)
    DriverAxis=(X=-1,Y=0,Z=0)
    DriverItem=(Type=Bone,Name="upperarm_r")
    FalloffRegionScaleFactors=(PositiveWidth=1,NegativeWidth=1,PositiveHeight=1,NegativeHeight=1)
    FalloffSize=0.2
    FlipHeightScaling=False
    FlipWidthScaling=False
    OptionalParentItem=(Type=Bone,Name="None")
    RotationOffset=(X=-119.3316,Y=0,Z=19.1823)
   )
   > &RigUnit_ModifyTransforms(
    ItemToModify="((Item=(Type=Bone,Name=\"dyn_hair_back_r_MSTR\"),Transform=(Rotation=(X=-0.1767,Y=0.0395,Z=0.0931,W=0.9791),Translation=(X=-1,Y=1.23,Z=1.5194),Scale3D=(X=1,Y=1,Z=1))),(Item=(Type=Bone,Name=\"dyn_hair_front_r_MSTR\"),Transform=(Rotation=(X=0.0095,Y=0.2718,Z=0.0 ...(truncated)"
    WeightMinimum=0
    Weight=@n9.OutputParam
   )
   > &RigUnit_SetCurveValue(Curve=arm_r_up_SPR, Value=@n9.OutputParam)
   > n10 = &RigUnit_SphericalPoseReader(
    ActiveRegionScaleFactors=(PositiveWidth=1,NegativeWidth=1,PositiveHeight=1,NegativeHeight=1)
    ActiveRegionSize=0.12
    Debug=(bDrawDebug=false,bDraw2D=False,bDrawLocalAxes=False,DebugScale=25,DebugSegments=20,DebugThickness=0.25)
    DriverAxis=(X=1,Y=0,Z=0)
    DriverItem=(Type=Bone,Name="upperarm_l")
    FalloffRegionScaleFactors=(PositiveWidth=2,NegativeWidth=2,PositiveHeight=1,NegativeHeight=1)
    FalloffSize=0.2
    FlipHeightScaling=False
    FlipWidthScaling=False
    OptionalParentItem=(Type=Bone,Name="None")
    RotationOffset=(X=62.7363,Y=26.5304,Z=-31.8961)
   )
   > &RigUnit_ModifyTransforms(
    ItemToModify=((Item=(Type=Bone,Name="dyn_hair_front_l_MSTR"),Transform=(Rotation=(X=-0.0057,Y=0.0433,Z=0.1304,W=0.9905),Translation=(X=0,Y=0,Z=0),Scale3D=(X=1,Y=1,Z=1))))
    WeightMinimum=0
    Weight=@n10.OutputParam
   )
   > &RigUnit_SetCurveValue(Curve=arm_l_up_SPR, Value=@n10.OutputParam)
   ItemArray = RigVMDispatchNode(Value=((Type=Control,Name="dyn_hair_back_l_01_ctrl"),(Type=Control,Name="dyn_hair_back_l_02_ctrl"),(Type=Control,Name="dyn_hair_back_l_03_ctrl"))) {template="DISPATCH_RigVMDispatch_Constant(io Value)"}
   ItemArray_1 = RigVMDispatchNode(Value=((Type=Bone,Name="dyn_hair_back_l_01"),(Type=Bone,Name="dyn_hair_back_l_02"),(Type=Bone,Name="dyn_hair_back_l_03"))) {template="DISPATCH_RigVMDispatch_Constant(io Value)"}
   > SplineIK = RigVMFunctionReferenceNode(
    Primary Axis=(X=1,Y=0,Z=0)
    Samples Per Segment=4
    Secondary Spline Direction=(X=0,Y=5,Z=0)
    Spline Mode=BSpline
    Up Axis=(X=0,Y=1,Z=0)
    debug=true
    Controls=@ItemArray.Value
    Bones=@ItemArray_1.Value
   )
   ItemArray_2 = RigVMDispatchNode(Value=((Type=Control,Name="dyn_hair_back_mid_01_ctrl"),(Type=Control,Name="dyn_hair_back_mid_02_ctrl"),(Type=Control,Name="dyn_hair_back_mid_03_ctrl"))) {template="DISPATCH_RigVMDispatch_Constant(io Value)"}
   ItemArray_3 = RigVMDispatchNode(Value=((Type=Bone,Name="dyn_hair_back_mid_01"),(Type=Bone,Name="dyn_hair_back_mid_02"),(Type=Bone,Name="dyn_hair_back_mid_03"))) {template="DISPATCH_RigVMDispatch_Constant(io Value)"}
   > SplineIK_1 = RigVMFunctionReferenceNode(
    Primary Axis=(X=1,Y=0,Z=0)
    Samples Per Segment=4
    Secondary Spline Direction=(X=0,Y=5,Z=0)
    Spline Mode=BSpline
    Up Axis=(X=0,Y=1,Z=0)
    debug=true
    Controls=@ItemArray_2.Value
    Bones=@ItemArray_3.Value
   )
   ItemArray_4 = RigVMDispatchNode(Value=((Type=Control,Name="dyn_hair_back_r_01_ctrl"),(Type=Control,Name="dyn_hair_back_r_02_ctrl"),(Type=Control,Name="dyn_hair_back_r_03_ctrl"))) {template="DISPATCH_RigVMDispatch_Constant(io Value)"}
   ItemArray_5 = RigVMDispatchNode(Value=((Type=Bone,Name="dyn_hair_back_r_01"),(Type=Bone,Name="dyn_hair_back_r_02"),(Type=Bone,Name="dyn_hair_back_r_03"))) {template="DISPATCH_RigVMDispatch_Constant(io Value)"}
   > SplineIK_2 = RigVMFunctionReferenceNode(
    Primary Axis=(X=1,Y=0,Z=0)
    Samples Per Segment=4
    Secondary Spline Direction=(X=0,Y=5,Z=0)
    Spline Mode=BSpline
    Up Axis=(X=0,Y=1,Z=0)
    debug=true
    Controls=@ItemArray_4.Value
    Bones=@ItemArray_5.Value
   )
   &RigUnit_SphericalPoseReader(
    ActiveRegionScaleFactors=(PositiveWidth=1,NegativeWidth=1,PositiveHeight=1,NegativeHeight=1)
    ActiveRegionSize=0.12
    Debug=(bDrawDebug=false,bDraw2D=False,bDrawLocalAxes=False,DebugScale=25,DebugSegments=20,DebugThickness=0.25)
    DriverAxis=(X=-1,Y=0,Z=0)
    DriverItem=(Type=Bone,Name="upperarm_r")
    ExecutePin=()
    FalloffRegionScaleFactors=(PositiveWidth=1,NegativeWidth=1,PositiveHeight=1,NegativeHeight=1)
    FalloffSize=0.4
    FlipHeightScaling=False
    FlipWidthScaling=False
    OptionalParentItem=(Type=Bone,Name="None")
    RotationOffset=(X=-119.3316,Y=0,Z=19.1823)
   )
  ]}
 RigVMModel Construction Event Graph {nodes [
   n1 = &RigUnit_PrepareForExecution()
   RigVMFunction_Sequence = &RigVMFunction_Sequence(ExecuteContext=@n1.ExecutePin)
   ItemArray = RigVMDispatchNode(Value=((Type=Bone,Name="dyn_hair_back_l_01"),(Type=Bone,Name="dyn_hair_back_l_02"),(Type=Bone,Name="dyn_hair_back_l_03"))) {template="DISPATCH_RigVMDispatch_Constant(io Value)"}
   RigVMVariableNode(Value=@ItemArray.Value, ExecuteContext=@RigVMFunction_Sequence.A) {variable=HairBones_L}
   ItemArray_1 = RigVMDispatchNode(Value=((Type=Bone,Name="dyn_hair_back_mid_01"),(Type=Bone,Name="dyn_hair_back_mid_02"),(Type=Bone,Name="dyn_hair_back_mid_03"))) {template="DISPATCH_RigVMDispatch_Constant(io Value)"}
   > RigVMVariableNode(Value=@ItemArray_1.Value) {variable=HairBones_Mid}
   ItemArray_2 = RigVMDispatchNode(Value=((Type=Bone,Name="dyn_hair_back_r_01"),(Type=Bone,Name="dyn_hair_back_r_02"),(Type=Bone,Name="dyn_hair_back_r_03"))) {template="DISPATCH_RigVMDispatch_Constant(io Value)"}
   > RigVMVariableNode(Value=@ItemArray_2.Value) {variable=HairBones_R}
   ItemArray_3 = RigVMDispatchNode(Value=((Type=Control,Name="dyn_hair_back_l_01_ctrl"),(Type=Control,Name="dyn_hair_back_l_02_ctrl"),(Type=Control,Name="dyn_hair_back_l_03_ctrl"))) {template="DISPATCH_RigVMDispatch_Constant(io Value)"}
   > RigVMVariableNode(Value=@ItemArray_3.Value) {variable=HairControls_L}
   ItemArray_4 = RigVMDispatchNode(Value=((Type=Control,Name="dyn_hair_back_mid_01_ctrl"),(Type=Control,Name="dyn_hair_back_mid_02_ctrl"),(Type=Control,Name="dyn_hair_back_mid_03_ctrl"))) {template="DISPATCH_RigVMDispatch_Constant(io Value)"}
   > RigVMVariableNode(Value=@ItemArray_4.Value) {variable=HairControls_Mid}
   ItemArray_5 = RigVMDispatchNode(Value=((Type=Control,Name="dyn_hair_back_r_01_ctrl"),(Type=Control,Name="dyn_hair_back_r_02_ctrl"),(Type=Control,Name="dyn_hair_back_r_03_ctrl"))) {template="DISPATCH_RigVMDispatch_Constant(io Value)"}
   > RigVMVariableNode(Value=@ItemArray_5.Value) {variable=HairControls_R}
   For_Each = RigVMDispatchNode(Array=RigVMVariableNode() {variable=HairBones_L}.Value, ExecuteContext=@RigVMFunction_Sequence.B) {template="DISPATCH_RigVMDispatch_ArrayIterator(in Array,out Element,out Index,out Count,out Ratio)"}
   > &RigUnit_SetTransform(
    Weight=1
    bInitial=true
    Item=RigVMDispatchNode(Index=@For_Each.Index, Array=RigVMVariableNode() {variable=HairControls_L}.Value) {template="DISPATCH_RigVMDispatch_ArrayGetAtIndex(in Array,in Index,out Element)"}.Element
    Value=&RigUnit_GetTransform(bInitial=true, Item=@For_Each.Element).Transform
   )
   > For_Each_1 = RigVMDispatchNode(Array=RigVMVariableNode() {variable=HairBones_Mid}.Value) {template="DISPATCH_RigVMDispatch_ArrayIterator(in Array,out Element,out Index,out Count,out Ratio)"}
   > &RigUnit_SetTransform(
    Weight=1
    bInitial=true
    Item=RigVMDispatchNode(Index=@For_Each_1.Index, Array=RigVMVariableNode() {variable=HairControls_Mid}.Value) {template="DISPATCH_RigVMDispatch_ArrayGetAtIndex(in Array,in Index,out Element)"}.Element
    Value=&RigUnit_GetTransform(bInitial=true, Item=@For_Each_1.Element).Transform
   )
   > For_Each_2 = RigVMDispatchNode(Array=RigVMVariableNode() {variable=HairBones_R}.Value) {template="DISPATCH_RigVMDispatch_ArrayIterator(in Array,out Element,out Index,out Count,out Ratio)"}
   > &RigUnit_SetTransform(
    Weight=1
    bInitial=true
    Item=RigVMDispatchNode(Index=@For_Each_2.Index, Array=RigVMVariableNode() {variable=HairControls_R}.Value) {template="DISPATCH_RigVMDispatch_ArrayGetAtIndex(in Array,in Index,out Element)"}.Element
    Value=&RigUnit_GetTransform(bInitial=true, Item=@For_Each_2.Element).Transform
   )
  ]}
]
hierarchy [
 root {loc=<0 0 0>, rot=<0 0 0 1>}
 pelvis {parent=root, loc=<0 -0.7071 93.1191>, rot=<0.0127 0.707 -0.0127 -0.707>}
 spine_01 {parent=pelvis, loc=<6.7494 -0.6388 0>, rot=<0 0 0.034 -0.9994>}
 spine_02 {parent=spine_01, loc=<7.2518 -0.035 0>, rot=<0 0 0.0238 -0.9997>}
 spine_03 {parent=spine_02, loc=<6.0104 0 0>, rot=<0 0 0.0149 -0.9999>}
 spine_04 {parent=spine_03, loc=<8.1024 0 0>, rot=<0 0 -0.0132 -0.9999>}
 spine_05 {parent=spine_04, loc=<10.0146 0 0>, rot=<0 0 -0.0015 -1>}
 neck_01 {parent=spine_05, loc=<11.3138 0.1529 0>, rot=<0 0 -0.2003 -0.9797>}
 neck_02 {parent=neck_01, loc=<4.5627 0 0>, rot=<0 0 0.0504 -0.9987>}
 head {parent=neck_02, loc=<5.3914 0 0>, rot=<0 0 0.115 -0.9934>}
 dyn_hair_back_mid_MSTR {parent=head, loc=<10.8432 -9.3631 -0.6465>, rot=<0.0774 -0.7029 0.7029 -0.0774>}
 dyn_hair_back_mid_01 {parent=dyn_hair_back_mid_MSTR, loc=<0 0 0>, rot=<0 0 0 1>}
 dyn_hair_back_mid_02 {parent=dyn_hair_back_mid_01, loc=<11.5655 0 0>, rot=<0 0.0093 0 -1>}
 dyn_hair_back_mid_03 {parent=dyn_hair_back_mid_02, loc=<8.4645 0 0>, rot=<0 0 0 1>}
 dyn_hair_back_mid_strand_02 {parent=dyn_hair_back_mid_02, loc=<6.5441 -2.5478 1.1996>, rot=<0.1098 0.1428 -0.197 -0.9637>}
 dyn_hair_back_mid_strand_01 {parent=dyn_hair_back_mid_01, loc=<5.6344 -0.271 1.2442>, rot=<-0.1466 0.4725 0.1269 -0.8597>}
 dyn_hair_back_l_strand_02 {parent=dyn_hair_back_mid_01, loc=<10.6024 6.0654 -0.7512>, rot=<0.2034 0.2191 0.0462 -0.9531>}
 dyn_hair_back_r_strand_01 {parent=dyn_hair_back_mid_01, loc=<10.4573 -6.7897 -0.5214>, rot=<-0.0996 0.5984 0.0084 -0.7949>}
 dyn_hair_back_r_MSTR {parent=head, loc=<10.4683 -5.7874 6.3266>, rot=<0.093 -0.2841 0.9538 -0.0313>}
 dyn_hair_back_r_01 {parent=dyn_hair_back_r_MSTR, loc=<0 0 0>, rot=<0 0 0 1>}
 dyn_hair_back_r_02 {parent=dyn_hair_back_r_01, loc=<10.1577 0 0>, rot=<0.0122 0.0934 -0.0036 -0.9955>}
 dyn_hair_back_r_03 {parent=dyn_hair_back_r_02, loc=<5.7932 0 0>, rot=<0 0 0 1>}
 dyn_hair_back_r_strand_03 {parent=dyn_hair_back_r_02, loc=<4.5664 -2.8342 1.3399>, rot=<0.0166 0.4482 0.1225 -0.8854>}
 dyn_hair_back_r_strand_02 {parent=dyn_hair_back_r_01, loc=<9.1445 -1.2984 2.6775>, rot=<-0.0707 0.4509 0.1703 -0.8733>}
 dyn_hair_back_l_MSTR {parent=head, loc=<10.7079 -6.2337 -5.1311>, rot=<0.0509 -0.9523 0.2281 -0.1964>}
 dyn_hair_back_l_01 {parent=dyn_hair_back_l_MSTR, loc=<0 0 0>, rot=<0 0 0 1>}
 dyn_hair_back_l_02 {parent=dyn_hair_back_l_01, loc=<10.2755 0 0>, rot=<-0.0055 -0.0172 0.0022 -0.9998>}
 dyn_hair_back_l_03 {parent=dyn_hair_back_l_02, loc=<6.2146 0 0>, rot=<0 0 0 1>}
 dyn_hair_back_l_strand_01 {parent=dyn_hair_back_l_01, loc=<8.6316 0.3985 0.813>, rot=<-0.048 0.2636 0.1805 -0.9464>}
 dyn_hair_front_r_MSTR {parent=head, loc=<5.6934 -0.4491 7.7728>, rot=<0.0701 0.1189 0.9889 0.0548>}
 dyn_hair_front_r_01 {parent=dyn_hair_front_r_MSTR, loc=<0 0 0>, rot=<0 0 0 1>}
 dyn_hair_front_r_02 {parent=dyn_hair_front_r_01, loc=<7.439 0 0>, rot=<-0.2847 0.0336 0.1315 -0.949>}
 dyn_hair_front_r_03 {parent=dyn_hair_front_r_02, loc=<7.3379 0 0>, rot=<-0.225 0 0 -0.9744>}
 dyn_hair_front_r_strand_02 {parent=dyn_hair_front_r_02, loc=<3.3332 -3.0127 2.1369>, rot=<-0.1922 -0.0316 0.071 -0.9783>}
 dyn_hair_front_l_MSTR {parent=head, loc=<5.6937 -0.4491 -7.7728>, rot=<-0.0439 -0.9799 -0.1674 -0.099>}
 dyn_hair_front_l_01 {parent=dyn_hair_front_l_MSTR, loc=<0 0 0>, rot=<0 0 0 1>}
 dyn_hair_front_l_02 {parent=dyn_hair_front_l_01, loc=<7.545 0 0>, rot=<0.2007 -0.0424 -0.1977 -0.9586>}
 dyn_hair_front_l_03 {parent=dyn_hair_front_l_02, loc=<7.5828 0 0>, rot=<0.342 0 0 -0.9397>}
 dyn_hair_front_l_strand_05 {parent=dyn_hair_front_l_02, loc=<3.0165 2.5203 4.3252>, rot=<-0.0856 0.3139 0.2566 -0.9101>}
 dyn_hair_front_l_strand_06 {parent=dyn_hair_front_l_02, loc=<7.0122 1.3166 2.3742>, rot=<0.4897 0.3204 -0.2385 -0.775>}
 dyn_hair_front_l_strand_04 {parent=dyn_hair_front_l_01, loc=<1.694 2.6866 2.5902>, rot=<-0.0473 0.5016 0.0839 -0.8597>}
 dyn_hair_front_l_strand_02 {parent=head, loc=<10.5588 9.6371 -4.6604>, rot=<-0.2207 -0.838 -0.1531 -0.4751>}
 dyn_hair_front_l_strand_01 {parent=head, loc=<12.6891 7.4288 -4.8658>, rot=<-0.1221 -0.627 -0.1056 -0.7621>}
 dyn_hair_front_l_strand_03 {parent=head, loc=<12.0654 2.33 -7.0656>, rot=<0.2653 -0.5849 0.2059 -0.7383>}
 dyn_hair_front_r_strand_01 {parent=head, loc=<10.2622 2.8303 6.9947>, rot=<0.4401 -0.2336 0.8581 -0.124>}
 faceAttach {parent=head, loc=<1.5213 0.7823 0>, rot=<0.0029 -0.7071 -0.0029 -0.7071>}
 deform_glasses {parent=head, loc=<8.3195 1.1863 0>, rot=<-0.3976 -0.5847 -0.5847 -0.3976>}
 clavicle_l {parent=spine_05, loc=<4.3575 -0.1466 -2.3223>, rot=<0.0465 -0.7289 -0.009 -0.6829>}
 upperarm_l {parent=clavicle_l, loc=<15.5111 0 0>, rot=<0.0226 -0.3234 -0.0088 -0.9459>}
 upperarm_twist_01_l {parent=upperarm_l, loc=<0.0041 0 0>, rot=<0 0 0 1>}
 deform_upperarm_bk_base_l {parent=upperarm_twist_01_l, loc=<1.0813 -7.2193 -1.9453>, rot=<0 0 0 1>}
 deform_upperarm_bk_l {parent=deform_upperarm_bk_base_l, loc=<0 0 0>, rot=<0 0 0 1>}
 deform_upperarm_fr_base_l {parent=upperarm_twist_01_l, loc=<-0.3691 7.0506 -2.765>, rot=<0 0 0 1>}
 deform_upperarm_fr_l {parent=deform_upperarm_fr_base_l, loc=<0 0 0>, rot=<0 0 0 1>}
 deform_upperarm_up_base_l {parent=upperarm_twist_01_l, loc=<-3.2216 -1.1391 3.3311>, rot=<0 0 0 1>}
 deform_upperarm_up_l {parent=deform_upperarm_up_base_l, loc=<0 0 0>, rot=<0 0 0 1>}
 upperarm_twist_02_l {parent=upperarm_l, loc=<13.7941 0 0>, rot=<0 0 0 1>}
 clavicle_r {parent=spine_05, loc=<4.3575 -0.1466 2.3223>, rot=<0.7289 0.0465 -0.6829 0.009>}
 upperarm_r {parent=clavicle_r, loc=<-15.5111 0 0>, rot=<0.0226 -0.3234 -0.0088 -0.9459>}
 upperarm_twist_01_r {parent=upperarm_r, loc=<-0.0041 0 0>, rot=<0 0 0 1>}
 deform_upperarm_bk_base_r {parent=upperarm_twist_01_r, loc=<-1.0814 7.2192 1.9455>, rot=<0 0 0 1>}
 deform_upperarm_bk_r {parent=deform_upperarm_bk_base_r, loc=<0 0 0>, rot=<0 0 0 1>}
 deform_upperarm_fr_base_r {parent=upperarm_twist_01_r, loc=<0.3689 -7.0506 2.7652>, rot=<0 0 0 1>}
 deform_upperarm_fr_r {parent=deform_upperarm_fr_base_r, loc=<0 0 0>, rot=<0 0 0 1>}
 deform_upperarm_up_base_r {parent=upperarm_twist_01_r, loc=<3.2214 1.1391 -3.3308>, rot=<0 0 0 1>}
 deform_upperarm_up_r {parent=deform_upperarm_up_base_r, loc=<0 0 0>, rot=<0 0 0 -1>}
 upperarm_twist_02_r {parent=upperarm_r, loc=<-13.7941 0 0>, rot=<0 0 0 1>}
 arm_l_up_SPR {type=Curve, loc=<0 0 0>, rot=<0 0 0 1>}
 arm_r_up_SPR {type=Curve, loc=<0 0 0>, rot=<0 0 0 1>}
 dyn_hair_back_r_01_null {type=Null, loc=<-6.3266 -7.5611 162.4543>, rot=<-0.6648 -0.1814 0.7023 -0.1786>}
 dyn_hair_back_r_01_ctrl {controlType=EULER_TRANSFORM, parent=dyn_hair_back_r_01_null, type=Control, loc=<0 0 0>, rot=<0 0 0 1>}
 dyn_hair_back_r_02_null {type=Null, loc=<-7.9483 -8.6229 152.4833>, rot=<-0.5357 -0.1672 0.7901 -0.2466>}
 dyn_hair_back_r_02_ctrl {controlType=EULER_TRANSFORM, parent=dyn_hair_back_r_02_null, type=Control, loc=<0 0 0>, rot=<0 0 0 1>}
 dyn_hair_back_r_03_null {type=Null, loc=<-9.7114 -9.8425 147.1014>, rot=<-0.5357 -0.1672 0.7901 -0.2466>}
 dyn_hair_back_r_03_ctrl {controlType=EULER_TRANSFORM, parent=dyn_hair_back_r_03_null, type=Control, loc=<0 0 0>, rot=<0 0 0 1>}
 dyn_hair_back_l_01_null {type=Null, loc=<6.4527 -8.6835 159.2285>, rot=<-0.1136 -0.6457 0.2039 -0.7271>}
 dyn_hair_back_l_01_ctrl {controlType=EULER_TRANSFORM, parent=dyn_hair_back_l_01_null, type=Control, loc=<0 0 0>, rot=<0 0 0 1>}
 dyn_hair_back_l_02_null {type=Null, loc=<8.7369 -9.8486 153.2455>, rot=<-0.1352 -0.5484 0.1976 -0.8012>}
 dyn_hair_back_l_02_ctrl {controlType=EULER_TRANSFORM, parent=dyn_hair_back_l_02_null, type=Control, loc=<0 0 0>, rot=<0 0 0 1>}
 dyn_hair_back_l_03_null {type=Null, loc=<10.7287 -10.8946 147.4524>, rot=<-0.1352 -0.5484 0.1976 -0.8012>}
 dyn_hair_back_l_03_ctrl {controlType=EULER_TRANSFORM, parent=dyn_hair_back_l_03_null, type=Control, loc=<0 0 0>, rot=<0 0 0 1>}
 dyn_hair_back_mid_01_null {type=Null, loc=<0.6465 -11.8602 159.4374>, rot=<-0.3999 -0.3821 0.5736 -0.6042>}
 dyn_hair_back_mid_01_ctrl {controlType=EULER_TRANSFORM, parent=dyn_hair_back_mid_01_null, type=Control, loc=<0 0 0>, rot=<0 0 0 1>}
 dyn_hair_back_mid_02_null {type=Null, loc=<0.6465 -13.5629 151.4909>, rot=<-0.4394 -0.4394 0.554 -0.554>}
 dyn_hair_back_mid_02_ctrl {controlType=EULER_TRANSFORM, parent=dyn_hair_back_mid_02_null, type=Control, loc=<0 0 0>, rot=<0 0 0 1>}
 dyn_hair_back_mid_03_null {type=Null, loc=<0.6465 -15.4907 143.2488>, rot=<-0.4394 -0.4394 0.554 -0.554>}
 dyn_hair_back_mid_03_ctrl {controlType=EULER_TRANSFORM, parent=dyn_hair_back_mid_03_null, type=Control, loc=<0 0 0>, rot=<0 0 0 1>}
]
variables [
 HairBones_L {
  type="TArray<FRigElementKey>"
  default="((Type=Bone,Name=\"dyn_ear_l_01\"),(Type=Bone,Name=\"dyn_ear_l_02\"),(Type=Bone,Name=\"dyn_ear_l_03\"),(Type=Bone,Name=\"dyn_ear_l_04\"),(Type=Bone,Name=\"dyn_ear_l_05\"),(Type=Bone,Name=\"dyn_ear_l_06\"),(Type=Bone,Name=\"dyn_ear_l_07\"),(Type=Bone,Name=\"dyn_ear_l_08\") ...(truncated)"
 }
 HairBones_Mid {
  type="TArray<FRigElementKey>"
  default="((Type=Bone,Name=\"dyn_ear_l_01\"),(Type=Bone,Name=\"dyn_ear_l_02\"),(Type=Bone,Name=\"dyn_ear_l_03\"),(Type=Bone,Name=\"dyn_ear_l_04\"),(Type=Bone,Name=\"dyn_ear_l_05\"),(Type=Bone,Name=\"dyn_ear_l_06\"),(Type=Bone,Name=\"dyn_ear_l_07\"),(Type=Bone,Name=\"dyn_ear_l_08\") ...(truncated)"
 }
 HairBones_R {
  type="TArray<FRigElementKey>"
  default="((Type=Bone,Name=\"dyn_ear_l_01\"),(Type=Bone,Name=\"dyn_ear_l_02\"),(Type=Bone,Name=\"dyn_ear_l_03\"),(Type=Bone,Name=\"dyn_ear_l_04\"),(Type=Bone,Name=\"dyn_ear_l_05\"),(Type=Bone,Name=\"dyn_ear_l_06\"),(Type=Bone,Name=\"dyn_ear_l_07\"),(Type=Bone,Name=\"dyn_ear_l_08\") ...(truncated)"
 }
 HairControls_L {
  type="TArray<FRigElementKey>"
  default="((Type=Bone,Name=\"dyn_ear_l_01\"),(Type=Bone,Name=\"dyn_ear_l_02\"),(Type=Bone,Name=\"dyn_ear_l_03\"),(Type=Bone,Name=\"dyn_ear_l_04\"),(Type=Bone,Name=\"dyn_ear_l_05\"),(Type=Bone,Name=\"dyn_ear_l_06\"),(Type=Bone,Name=\"dyn_ear_l_07\"),(Type=Bone,Name=\"dyn_ear_l_08\") ...(truncated)"
 }
 HairControls_Mid {
  type="TArray<FRigElementKey>"
  default="((Type=Bone,Name=\"dyn_ear_l_01\"),(Type=Bone,Name=\"dyn_ear_l_02\"),(Type=Bone,Name=\"dyn_ear_l_03\"),(Type=Bone,Name=\"dyn_ear_l_04\"),(Type=Bone,Name=\"dyn_ear_l_05\"),(Type=Bone,Name=\"dyn_ear_l_06\"),(Type=Bone,Name=\"dyn_ear_l_07\"),(Type=Bone,Name=\"dyn_ear_l_08\") ...(truncated)"
 }
 HairControls_R {
  type="TArray<FRigElementKey>"
  default="((Type=Bone,Name=\"dyn_ear_l_01\"),(Type=Bone,Name=\"dyn_ear_l_02\"),(Type=Bone,Name=\"dyn_ear_l_03\"),(Type=Bone,Name=\"dyn_ear_l_04\"),(Type=Bone,Name=\"dyn_ear_l_05\"),(Type=Bone,Name=\"dyn_ear_l_06\"),(Type=Bone,Name=\"dyn_ear_l_07\"),(Type=Bone,Name=\"dyn_ear_l_08\") ...(truncated)"
 }
 SkydiveAlpha {type=double, default=0}
 useSplineHair {type=bool, default=False}
]
