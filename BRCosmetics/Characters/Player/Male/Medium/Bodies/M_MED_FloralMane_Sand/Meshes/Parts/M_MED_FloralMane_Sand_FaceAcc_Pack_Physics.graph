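# Physics asset for the FloralMane Sand face accessory pack on the medium male player body, setting up physics for its dynamic hair bones.
# Five hair strands (back left, middle, right; front left, right) each chain three capsules from a MSTR root body through angle-limited constraints.
# Unconstrained capsules on head, neck, clavicles and upper spine supply body shapes around the hair.
# Limits are tight, stiffest at back middle, with front tips freer on one swing axis.
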PhysicsAsset /BRCosmetics/Characters/Player/Male/Medium/Bodies/M_MED_FloralMane_Sand/Meshes/Parts/M_MED_FloralMane_Sand_FaceAcc_Pack_Physics
bodies {
 clavicle_l [{
   shapeType=Capsule
   length=10.8237
   radius=12.5093
   shapeName=clavicle_l_capsule
   center=<4.1566 0.2084 -4.1494>
   rotation=<rot 78.5145 -3.2308 -9.5684>
  }]
 clavicle_r [{
   shapeType=Capsule
   length=10.8237
   radius=12.5093
   shapeName=clavicle_l_capsule
   center=<-4.0228 -0.0571 4.1533>
   rotation=<rot 83.5703 -5.1301 -13.9948>
  }]
 dyn_hair_back_l_01 [{
   shapeType=Capsule
   length=6.6622
   radius=4.3378
   shapeName=dyn_hair_back_l_01_capsule
   center=<6.9743 0 3.3>
   rotation=<rot 0 0.0001 88.8893>
  }]
 dyn_hair_back_l_02 [{
   shapeType=Capsule
   length=9.0513
   radius=4.5
   shapeName=dyn_hair_back_l_02_capsule
   center=<4.0426 0 3.3>
   rotation=<rot -0.0001 0.0001 79.8876>
  }]
 dyn_hair_back_l_03 [{
   shapeType=Capsule
   length=13.0036
   radius=4.5797
   shapeName=dyn_hair_back_l_03_capsule
   center=<4.3751 0 2.5>
   rotation=<rot 1.57 5.6164 80.4715>
  }]
 dyn_hair_back_l_MSTR [{shapeType=Capsule, length=0.505, radius=0.505, shapeName=dyn_hair_back_l_MSTR_capsule}]
 dyn_hair_back_mid_01 [{shapeType=Capsule, length=8, radius=5, shapeName=dyn_hair_back_mid_01_capsule, center=<9.3337 0 3.3>, rotation=<rot 0 0 90>}]
 dyn_hair_back_mid_02 [{shapeType=Capsule, length=14.4215, radius=5.4592, shapeName=dyn_hair_back_mid_02_capsule, center=<3.9095 0 2.5804>, rotation=<rot 0 0 90>}]
 dyn_hair_back_mid_03 [{shapeType=Capsule, length=15.989, radius=5.3204, shapeName=dyn_hair_back_mid_03_capsule, center=<2.9108 0 2.761>, rotation=<rot 0 0 90>}]
 dyn_hair_back_mid_MSTR [{shapeType=Capsule, length=0.505, radius=0.505, shapeName=dyn_hair_back_mid_MSTR_capsule}]
 dyn_hair_back_r_01 [{shapeType=Capsule, length=6, radius=5, shapeName=dyn_hair_back_r_01_capsule, center=<7.7579 0 3.4>, rotation=<rot 0 0 90>}]
 dyn_hair_back_r_02 [{
   shapeType=Capsule
   length=10.584
   radius=4.9761
   shapeName=dyn_hair_back_r_02_capsule
   center=<4.9319 -0.1712 1.8506>
   rotation=<rot 0 0 93.8411>
  }]
 dyn_hair_back_r_03 [{
   shapeType=Capsule
   length=12.4402
   radius=4.6594
   shapeName=dyn_hair_back_r_03_capsule
   center=<2.2326 -0.8081 2.5085>
   rotation=<rot 4.7989 -15.433 88.7403>
  }]
 dyn_hair_back_r_MSTR [{shapeType=Capsule, length=0.505, radius=0.505, shapeName=dyn_hair_back_r_MSTR_capsule}]
 dyn_hair_front_l_01 [{shapeType=Capsule, length=6, radius=5, shapeName=dyn_hair_front_l_01_capsule, center=<4.0426 0 3.4>, rotation=<rot 90 0 0>}]
 dyn_hair_front_l_02 [{
   shapeType=Capsule
   length=6.101
   radius=4.399
   shapeName=dyn_hair_front_l_02_capsule
   center=<4.8372 -0.5548 3.5626>
   rotation=<rot 78.7827 -161.4457 168.2639>
  }]
 dyn_hair_front_l_03 [{
   shapeType=Capsule
   length=5.9351
   radius=4.0649
   shapeName=dyn_hair_front_l_03_capsule
   center=<3.0684 0 5.1853>
   rotation=<rot 77.635 179.9999 179.9999>
  }]
 dyn_hair_front_l_MSTR [{shapeType=Capsule, length=0.505, radius=0.505, shapeName=dyn_hair_front_l_MSTR_capsule}]
 dyn_hair_front_r_01 [{shapeType=Capsule, length=6, radius=5, shapeName=dyn_hair_front_r_01_capsule, center=<4.0426 0 3.4>, rotation=<rot 90 0 0>}]
 dyn_hair_front_r_02 [{
   shapeType=Capsule
   length=6
   radius=4.5
   shapeName=dyn_hair_front_r_02_capsule
   center=<3.5377 0 2.8767>
   rotation=<rot 80.0861 180 -163.8676>
  }]
 dyn_hair_front_r_03 [{
   shapeType=Capsule
   length=5.6733
   radius=4.3267
   shapeName=dyn_hair_front_r_03_capsule
   center=<2.9982 0 5.3535>
   rotation=<rot 77.4278 -173.1642 -164.0357>
  }]
 dyn_hair_front_r_MSTR [{shapeType=Capsule, length=0.505, radius=0.505, shapeName=dyn_hair_front_r_MSTR_capsule}]
 head [{shapeType=Capsule, length=1.8106, radius=1.3302, shapeName=head_capsule}]
 neck_01 [{shapeType=Capsule, length=12.3488, radius=7.2819, shapeName=neck_01_capsule, center=<2.6262 -1.4033 0>, rotation=<rot -90 0 0>}]
 neck_02 [{shapeType=Capsule, length=0.505, radius=0.505, shapeName=neck_02_capsule}]
 spine_05 [{shapeType=Capsule, length=15.6379, radius=11.5501, shapeName=spine_05_capsule, center=<0.6942 -1.0898 0>}]
}
constraints [
 {bone1Name=dyn_hair_back_l_01, bone2Name=dyn_hair_back_l_MSTR, swing1LimitDegrees=5, swing2LimitDegrees=5, twistLimitDegrees=10}
 {bone1Name=dyn_hair_back_l_02, bone2Name=dyn_hair_back_l_01, swing1LimitDegrees=5, swing2LimitDegrees=5, twistLimitDegrees=10}
 {bone1Name=dyn_hair_back_l_03, bone2Name=dyn_hair_back_l_02, swing1LimitDegrees=5, swing2LimitDegrees=5, twistLimitDegrees=10}
 {bone1Name=dyn_hair_back_mid_01, bone2Name=dyn_hair_back_mid_MSTR, swing1LimitDegrees=2, swing2LimitDegrees=2, twistLimitDegrees=5}
 {bone1Name=dyn_hair_back_mid_02, bone2Name=dyn_hair_back_mid_01, swing1LimitDegrees=2, swing2LimitDegrees=2, twistLimitDegrees=5}
 {bone1Name=dyn_hair_back_mid_03, bone2Name=dyn_hair_back_mid_02, swing1LimitDegrees=2, swing2LimitDegrees=2, twistLimitDegrees=5}
 {bone1Name=dyn_hair_back_r_01, bone2Name=dyn_hair_back_r_MSTR, swing1LimitDegrees=5, swing2LimitDegrees=5, twistLimitDegrees=10}
 {bone1Name=dyn_hair_back_r_02, bone2Name=dyn_hair_back_r_01, swing1LimitDegrees=5, swing2LimitDegrees=5, twistLimitDegrees=10}
 {bone1Name=dyn_hair_back_r_03, bone2Name=dyn_hair_back_r_02, swing1LimitDegrees=5, swing2LimitDegrees=5, twistLimitDegrees=10}
 {bone1Name=dyn_hair_front_l_01, bone2Name=dyn_hair_front_l_MSTR, swing1LimitDegrees=5, swing2LimitDegrees=5, twistLimitDegrees=2}
 {bone1Name=dyn_hair_front_l_02, bone2Name=dyn_hair_front_l_01, swing1LimitDegrees=7, swing2LimitDegrees=7, twistLimitDegrees=2}
 {bone1Name=dyn_hair_front_l_03, bone2Name=dyn_hair_front_l_02, swing1LimitDegrees=5, swing2LimitDegrees=30, twistLimitDegrees=2}
 {bone1Name=dyn_hair_front_r_01, bone2Name=dyn_hair_front_r_MSTR, swing1LimitDegrees=5, swing2LimitDegrees=5, twistLimitDegrees=2}
 {bone1Name=dyn_hair_front_r_02, bone2Name=dyn_hair_front_r_01, swing1LimitDegrees=7, swing2LimitDegrees=7, twistLimitDegrees=2}
 {bone1Name=dyn_hair_front_r_03, bone2Name=dyn_hair_front_r_02, swing1LimitDegrees=5, swing2LimitDegrees=30, twistLimitDegrees=2}
]
