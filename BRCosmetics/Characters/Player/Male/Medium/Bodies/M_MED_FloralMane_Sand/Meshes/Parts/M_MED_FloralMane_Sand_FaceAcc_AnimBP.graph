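# Post-process-style AnimBP for the FloralMane Sand face accessory, simulating the character's hair.
# Two branches, selected by isWearingPack, copy head and body poses into a Control Rig, head-space RigidBody with separate physics assets, and front-hair constraints.
# The blended result runs through AnimDynamics strands; a separate back-hair chain is unconnected.
# Events cache meshes and pack state, and update skydive, zipline, grind-rail alphas and gravity-override values.
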AnimBlueprint /BRCosmetics/Characters/Player/Male/Medium/Bodies/M_MED_FloralMane_Sand/Meshes/Parts/M_MED_FloralMane_Sand_FaceAcc_AnimBP
target_skeleton=&/M_MED_FloralMane_Sand_FaceAcc_Skeleton
graphs [
 AnimGraph {nodes [
   n1 = Anim Dynamics - Bone: dyn_hair_back_l_01!(ComponentPose, Alpha=1, GravityOverride:Vector="0,0,-980", ExternalForce="0,0,0") -> (Pose) {
    angularDampingOverride=0.999
    angularSpringConstant=1400
    bAngularSpring=true
    bOverrideAngularDamping=true
    bUseGravityOverride=true
    simulationSpace=RootRelative
    boundBone {boneName=dyn_hair_back_l_01}
    componentAppliedLinearAccClamp=<3000 3000 3000>
    gravityOverride {z=-980}
    physicsBodyDefinitions [{
      boundBone {boneName=dyn_hair_back_l_01}
      boxExtents=<40 40 40>
      constraintSetup {angularLimitsMax=<1 0 1>, angularLimitsMin=<-1 -1 -1>, angularTarget=<1 0 0>}
      localJointOffset=<1 0 0>
     }]
    simSpaceSettings {maxAngularAcceleration=40, maxAngularVelocity=8, simSpaceAngularAlpha=1}
   }
   n2 = Anim Dynamics - Bone: dyn_hair_back_l_02!(ComponentPose=@n1, Alpha=1, GravityOverride:Vector="0,0,-980", ExternalForce="0,0,0") -> (Pose) {
    angularDampingOverride=0.999
    angularSpringConstant=1400
    bAngularSpring=true
    bOverrideAngularDamping=true
    bUseGravityOverride=true
    simulationSpace=RootRelative
    boundBone {boneName=dyn_hair_back_l_02}
    componentAppliedLinearAccClamp=<3000 3000 3000>
    gravityOverride {z=-980}
    physicsBodyDefinitions [{
      boundBone {boneName=dyn_hair_back_l_02}
      boxExtents=<40 40 40>
      constraintSetup {angularLimitsMax=<1 0 1>, angularLimitsMin=<-1 -1 -1>, angularTarget=<1 0 0>}
      localJointOffset=<1 0 0>
     }]
    simSpaceSettings {maxAngularAcceleration=40, maxAngularVelocity=8, simSpaceAngularAlpha=1}
   }
   n3 = Anim Dynamics - Bone: dyn_hair_back_l_03!(ComponentPose=@n2, Alpha=1, GravityOverride:Vector="0,0,-980", ExternalForce="0,0,0") -> (Pose) {
    angularDampingOverride=0.999
    angularSpringConstant=1400
    bAngularSpring=true
    bOverrideAngularDamping=true
    bUseGravityOverride=true
    simulationSpace=RootRelative
    boundBone {boneName=dyn_hair_back_l_03}
    componentAppliedLinearAccClamp=<3000 3000 3000>
    gravityOverride {z=-980}
    physicsBodyDefinitions [{
      boundBone {boneName=dyn_hair_back_l_03}
      boxExtents=<40 40 40>
      constraintSetup {angularLimitsMax=<1 0 1>, angularLimitsMin=<-1 -1 -1>, angularTarget=<1 0 0>}
      localJointOffset=<1 0 0>
     }]
    simSpaceSettings {maxAngularAcceleration=40, maxAngularVelocity=8, simSpaceAngularAlpha=1}
   }
   n4 = Anim Dynamics - Bone: dyn_hair_back_mid_01!(ComponentPose=@n3, Alpha=1, GravityOverride:Vector="0,0,-980", ExternalForce="0,0,0") -> (Pose) {
    angularDampingOverride=0.999
    angularSpringConstant=1400
    bAngularSpring=true
    bOverrideAngularDamping=true
    bUseGravityOverride=true
    simulationSpace=RootRelative
    boundBone {boneName=dyn_hair_back_mid_01}
    componentAppliedLinearAccClamp=<3000 3000 3000>
    gravityOverride {z=-980}
    physicsBodyDefinitions [{
      boundBone {boneName=dyn_hair_back_mid_01}
      boxExtents=<40 40 40>
      constraintSetup {angularLimitsMax=<1 0 1>, angularLimitsMin=<-1 -1 -1>, angularTarget=<1 0 0>}
      localJointOffset=<1 0 0>
     }]
    simSpaceSettings {maxAngularAcceleration=40, maxAngularVelocity=8, simSpaceAngularAlpha=1}
   }
   n5 = Anim Dynamics - Bone: dyn_hair_back_mid_02!(ComponentPose=@n4, Alpha=1, GravityOverride:Vector="0,0,-980", ExternalForce="0,0,0") -> (Pose) {
    angularDampingOverride=0.999
    angularSpringConstant=1400
    bAngularSpring=true
    bOverrideAngularDamping=true
    bUseGravityOverride=true
    simulationSpace=RootRelative
    boundBone {boneName=dyn_hair_back_mid_02}
    componentAppliedLinearAccClamp=<3000 3000 3000>
    gravityOverride {z=-980}
    physicsBodyDefinitions [{
      boundBone {boneName=dyn_hair_back_mid_02}
      boxExtents=<40 40 40>
      constraintSetup {angularLimitsMax=<1 0 1>, angularLimitsMin=<-1 -1 -1>, angularTarget=<1 0 0>}
      localJointOffset=<1 0 0>
     }]
    simSpaceSettings {maxAngularAcceleration=40, maxAngularVelocity=8, simSpaceAngularAlpha=1}
   }
   n6 = Anim Dynamics - Bone: dyn_hair_back_mid_03!(ComponentPose=@n5, Alpha=1, GravityOverride:Vector="0,0,-980", ExternalForce="0,0,0") -> (Pose) {
    angularDampingOverride=0.999
    angularSpringConstant=1400
    bAngularSpring=true
    bOverrideAngularDamping=true
    bUseGravityOverride=true
    simulationSpace=RootRelative
    boundBone {boneName=dyn_hair_back_mid_03}
    componentAppliedLinearAccClamp=<3000 3000 3000>
    gravityOverride {z=-980}
    physicsBodyDefinitions [{
      boundBone {boneName=dyn_hair_back_mid_03}
      boxExtents=<40 40 40>
      constraintSetup {angularLimitsMax=<1 0 1>, angularLimitsMin=<-1 -1 -1>, angularTarget=<1 0 0>}
      localJointOffset=<1 0 0>
     }]
    simSpaceSettings {maxAngularAcceleration=40, maxAngularVelocity=8, simSpaceAngularAlpha=1}
   }
   n7 = Anim Dynamics - Bone: dyn_hair_back_r_01!(ComponentPose=@n6, Alpha=1, GravityOverride:Vector="0,0,-980", ExternalForce="0,0,0") -> (Pose) {
    angularDampingOverride=0.999
    angularSpringConstant=1400
    bAngularSpring=true
    bOverrideAngularDamping=true
    bUseGravityOverride=true
    simulationSpace=RootRelative
    boundBone {boneName=dyn_hair_back_r_01}
    componentAppliedLinearAccClamp=<3000 3000 3000>
    gravityOverride {z=-980}
    physicsBodyDefinitions [{
      boundBone {boneName=dyn_hair_back_r_01}
      boxExtents=<40 40 40>
      constraintSetup {angularLimitsMax=<1 0 1>, angularLimitsMin=<-1 -1 -1>, angularTarget=<1 0 0>}
      localJointOffset=<1 0 0>
     }]
    simSpaceSettings {maxAngularAcceleration=40, maxAngularVelocity=8, simSpaceAngularAlpha=1}
   }
   n8 = Anim Dynamics - Bone: dyn_hair_back_r_02!(ComponentPose=@n7, Alpha=1, GravityOverride:Vector="0,0,-980", ExternalForce="0,0,0") -> (Pose) {
    angularDampingOverride=0.999
    angularSpringConstant=1400
    bAngularSpring=true
    bOverrideAngularDamping=true
    bUseGravityOverride=true
    simulationSpace=RootRelative
    boundBone {boneName=dyn_hair_back_r_02}
    componentAppliedLinearAccClamp=<3000 3000 3000>
    gravityOverride {z=-980}
    physicsBodyDefinitions [{
      boundBone {boneName=dyn_hair_back_r_02}
      boxExtents=<40 40 40>
      constraintSetup {angularLimitsMax=<1 0 1>, angularLimitsMin=<-1 -1 -1>, angularTarget=<1 0 0>}
      localJointOffset=<1 0 0>
     }]
    simSpaceSettings {maxAngularAcceleration=40, maxAngularVelocity=8, simSpaceAngularAlpha=1}
   }
   Anim Dynamics - Bone: dyn_hair_back_r_03!(ComponentPose=@n8, Alpha=1, GravityOverride:Vector="0,0,-980", ExternalForce="0,0,0") -> (Pose) {
    angularDampingOverride=0.999
    angularSpringConstant=1400
    bAngularSpring=true
    bOverrideAngularDamping=true
    bUseGravityOverride=true
    simulationSpace=RootRelative
    boundBone {boneName=dyn_hair_back_r_03}
    componentAppliedLinearAccClamp=<3000 3000 3000>
    gravityOverride {z=-980}
    physicsBodyDefinitions [{
      boundBone {boneName=dyn_hair_back_r_03}
      boxExtents=<40 40 40>
      constraintSetup {angularLimitsMax=<1 0 1>, angularLimitsMin=<-1 -1 -1>, angularTarget=<1 0 0>}
      localJointOffset=<1 0 0>
     }]
    simSpaceSettings {maxAngularAcceleration=40, maxAngularVelocity=8, simSpaceAngularAlpha=1}
   }
   n9 = Get HeadMesh() -> (HeadMesh:Skeletal Mesh Component Budgeted Object Reference)
   n10 = Copy Pose From Mesh() {bUseAttachedParent=true}
   n11 = Copy Pose From Mesh(SourceMeshComponent=Get CharacterMesh() -> (CharacterMesh:Skeletal Mesh Component Object Reference)) {bUseAttachedParent=true}
   n12 = Control Rig!(
    Source=Layered blend per bone(
     BasePose=Copy Pose From Mesh(SourceMeshComponent=@n9) {bUseAttachedParent=true}
     BlendPoses_0=Layered blend per bone(BasePose=@n11, BlendPoses_0=@n10)
    ) {layerSetup [{branchFilters [{boneName=upperarm_l}, {boneName=upperarm_r}]}]}
    Alpha
    useSplineHair:bool=Get isWearingPack() -> (isWearingPack:bool)
    SkydiveAlpha:double=Get skydiveAlpha() -> (skydiveAlpha:double)
   ) -> (Pose) {controlRigAssetReference {blueprintRigClass=&/M_MED_FloralMane_Sand_FaceAcc_CtrlRig.M_MED_FloralMane_Sand_FaceAcc_CtrlRig_C}}
   n13 = RigidBody!(
    ComponentPose=Local To Component(LocalPose=@n12)
    Alpha
    OverrideWorldGravity:Vector="0,0,-980"
    ExternalForce
    SimSpaceSettings:Sim Space Settings Structure=(WorldAlpha=1,VelocityScaleZ=1,DampingAlpha=0,MaxLinearVelocity=400,MaxAngularVelocity=6,MaxLinearAcceleration=2000,MaxAngularAcceleration=30,ExternalLinearDragV=(X=0.1,Y=0.1,Z=0.1),ExternalLinearVelocity=(X=0,Y=0,Z=0),ExternalAngularVelocity=(X=0,Y=0,Z=0))
   ) -> (Pose) {
    bClampLinearTranslationLimitToRefPose=true
    bOverrideWorldGravity=true
    overridePhysicsAsset=&/M_MED_FloralMane_Sand_FaceAcc_Physics
    simulationSpace=BaseBoneSpace
    baseBoneRef {boneName=head}
    overrideWorldGravity {z=-980}
    simSpaceSettings {
     dampingAlpha=0
     maxAngularAcceleration=30
     maxAngularVelocity=6
     maxLinearAcceleration=2000
     maxLinearVelocity=400
     worldAlpha=1
     externalLinearDragV=<0.1 0.1 0.1>
    }
   }
   n14 = "Constraint - dyn_hair_front_r_02 (PP)"!(ComponentPose=@n13, Alpha=1, ConstraintWeights_0=0.7, ConstraintWeights_1:float=0.3) -> (Pose) {
    alphaScaleBiasClamp {bInterpResult=true}
    boneToModify {boneName=dyn_hair_front_r_02}
    constraintSetup [{targetBone {boneName=spine_05}}, {targetBone {boneName=head}}]
    constraintWeights [1,1]
   }
   n15 = "Constraint - dyn_hair_front_l_02 (PP)"!(ComponentPose=@n14, Alpha=1, ConstraintWeights_0=0.7, ConstraintWeights_1:float=0.3) -> (Pose) {
    alphaScaleBiasClamp {bInterpResult=true}
    boneToModify {boneName=dyn_hair_front_l_02}
    constraintSetup [{targetBone {boneName=spine_05}}, {targetBone {boneName=head}}]
    constraintWeights [1,1]
   }
   n16 = Component To Local(ComponentPose=@n15)
   n17 = Get HeadMesh()
   n18 = Copy Pose From Mesh() {bUseAttachedParent=true}
   n19 = Copy Pose From Mesh(SourceMeshComponent=Get CharacterMesh()) {bUseAttachedParent=true}
   n20 = Control Rig(
    Source=Layered blend per bone(
     BasePose=Copy Pose From Mesh(SourceMeshComponent=@n17) {bUseAttachedParent=true}
     BlendPoses_0=Layered blend per bone(BasePose=@n19, BlendPoses_0=@n18)
    ) {layerSetup [{branchFilters [{boneName=upperarm_l}, {boneName=upperarm_r}]}]}
    useSplineHair=Get isWearingPack()
    SkydiveAlpha=Get skydiveAlpha()
   ) {controlRigAssetReference {blueprintRigClass=&M_MED_FloralMane_Sand_FaceAcc_CtrlRig_C}}
   n21 = RigidBody(
    ComponentPose=Local To Component(LocalPose=@n20)
    Alpha=0.6
    OverrideWorldGravity="0,0,-980"
    SimSpaceSettings=(WorldAlpha=1,VelocityScaleZ=1,DampingAlpha=0,MaxLinearVelocity=400,MaxAngularVelocity=6,MaxLinearAcceleration=2000,MaxAngularAcceleration=30,ExternalLinearDragV=(X=0.1,Y=0.1,Z=0.1),ExternalLinearVelocity=(X=0,Y=0,Z=0),ExternalAngularVelocity=(X=0,Y=0,Z=0))
   ) {
    bClampLinearTranslationLimitToRefPose=true
    bOverrideWorldGravity=true
    overridePhysicsAsset=&/M_MED_FloralMane_Sand_FaceAcc_Pack_Physics
    simulationSpace=BaseBoneSpace
    alphaScaleBiasClamp {bInterpResult=true}
    baseBoneRef {boneName=head}
    overrideWorldGravity {z=-980}
    simSpaceSettings {
     dampingAlpha=0
     maxAngularAcceleration=30
     maxAngularVelocity=6
     maxLinearAcceleration=2000
     maxLinearVelocity=400
     worldAlpha=1
     externalLinearDragV=<0.1 0.1 0.1>
    }
   }
   n22 = "Constraint - dyn_hair_front_r_02 (PP)"(ComponentPose=@n21, Alpha=1, ConstraintWeights_0=0.7, ConstraintWeights_1=0.3) {
    alphaScaleBiasClamp {bInterpResult=true}
    boneToModify {boneName=dyn_hair_front_r_02}
    constraintSetup [{targetBone {boneName=spine_05}}, {targetBone {boneName=head}}]
    constraintWeights [1,1]
   }
   n23 = "Constraint - dyn_hair_front_l_02 (PP)"(ComponentPose=@n22, Alpha=1, ConstraintWeights_0=0.7, ConstraintWeights_1=0.3) {
    alphaScaleBiasClamp {bInterpResult=true}
    boneToModify {boneName=dyn_hair_front_l_02}
    constraintSetup [{targetBone {boneName=spine_05}}, {targetBone {boneName=head}}]
    constraintWeights [1,1]
   }
   n24 = Component To Local(ComponentPose=@n23)
   n25 = Get isWearingPack()
   n26 = Anim Dynamics - Bone: dyn_hair_front_l_strand_01!(
    ComponentPose=Local To Component(LocalPose=Blend Poses by bool(BlendPose_0=@n24, BlendPose_1=@n16, bActiveValue=@n25))
    Alpha=1
    GravityOverride:Vector="0,0,-980"
    ExternalForce="0,0,0"
   ) -> (Pose) {
    angularDampingOverride=0.8
    angularSpringConstant=15000
    bAngularSpring=true
    bOverrideAngularDamping=true
    bUseGravityOverride=true
    simulationSpace=RootRelative
    alphaScaleBiasClamp {bInterpResult=true}
    boundBone {boneName=dyn_hair_front_l_strand_01}
    componentAppliedLinearAccClamp=<3000 3000 3000>
    componentLinearAccScale=<0.2 0.2 0.2>
    componentLinearVelScale=<0.2 0.2 0.2>
    gravityOverride {z=-980}
    physicsBodyDefinitions [{
      boundBone {boneName=dyn_hair_front_l_strand_01}
      boxExtents=<40 40 40>
      constraintSetup {angularLimitsMax=<0 10 10>, angularLimitsMin=<0 -10 -10>, angularTarget=<1 0 0>}
      localJointOffset=<1 0 0>
     }]
    simSpaceSettings {maxAngularAcceleration=60, maxAngularVelocity=12, simSpaceAngularAlpha=0.75}
   }
   n27 = Anim Dynamics - Bone: dyn_hair_front_l_strand_02!(ComponentPose=@n26, Alpha=1, GravityOverride:Vector="0,0,-980", ExternalForce="0,0,0") -> (Pose) {
    angularDampingOverride=0.8
    angularSpringConstant=15000
    bAngularSpring=true
    bOverrideAngularDamping=true
    bUseGravityOverride=true
    simulationSpace=RootRelative
    alphaScaleBiasClamp {bInterpResult=true}
    boundBone {boneName=dyn_hair_front_l_strand_02}
    componentAppliedLinearAccClamp=<3000 3000 3000>
    componentLinearAccScale=<0.2 0.2 0.2>
    componentLinearVelScale=<0.2 0.2 0.2>
    gravityOverride {z=-980}
    physicsBodyDefinitions [{
      boundBone {boneName=dyn_hair_front_l_strand_02}
      boxExtents=<40 40 40>
      constraintSetup {angularLimitsMax=<0 10 10>, angularLimitsMin=<0 -10 -10>, angularTarget=<1 0 0>}
      localJointOffset=<1 0 0>
     }]
    simSpaceSettings {maxAngularAcceleration=60, maxAngularVelocity=12, simSpaceAngularAlpha=0.75}
   }
   n28 = Anim Dynamics - Bone: dyn_hair_front_l_strand_03!(ComponentPose=@n27, Alpha=1, GravityOverride:Vector="0,0,-980", ExternalForce="0,0,0") -> (Pose) {
    angularDampingOverride=0.8
    angularSpringConstant=15000
    bAngularSpring=true
    bOverrideAngularDamping=true
    bUseGravityOverride=true
    simulationSpace=RootRelative
    alphaScaleBiasClamp {bInterpResult=true}
    boundBone {boneName=dyn_hair_front_l_strand_03}
    componentAppliedLinearAccClamp=<3000 3000 3000>
    componentLinearAccScale=<0.2 0.2 0.2>
    componentLinearVelScale=<0.2 0.2 0.2>
    gravityOverride {z=-980}
    physicsBodyDefinitions [{
      boundBone {boneName=dyn_hair_front_l_strand_03}
      boxExtents=<40 40 40>
      constraintSetup {angularLimitsMax=<0 10 10>, angularLimitsMin=<0 -10 -10>, angularTarget=<1 0 0>}
      localJointOffset=<1 0 0>
     }]
    simSpaceSettings {maxAngularAcceleration=60, maxAngularVelocity=12, simSpaceAngularAlpha=0.75}
   }
   n29 = Anim Dynamics - Bone: dyn_hair_front_l_strand_04!(ComponentPose=@n28, Alpha=1, GravityOverride:Vector="0,0,-980", ExternalForce="0,0,0") -> (Pose) {
    angularDampingOverride=0.8
    angularSpringConstant=15000
    bAngularSpring=true
    bOverrideAngularDamping=true
    bUseGravityOverride=true
    simulationSpace=RootRelative
    alphaScaleBiasClamp {bInterpResult=true}
    boundBone {boneName=dyn_hair_front_l_strand_04}
    componentAppliedLinearAccClamp=<3000 3000 3000>
    componentLinearAccScale=<0.2 0.2 0.2>
    componentLinearVelScale=<0.2 0.2 0.2>
    gravityOverride {z=-980}
    physicsBodyDefinitions [{
      boundBone {boneName=dyn_hair_front_l_strand_04}
      boxExtents=<40 40 40>
      constraintSetup {angularLimitsMax=<0 10 10>, angularLimitsMin=<0 -10 -10>, angularTarget=<1 0 0>}
      localJointOffset=<1 0 0>
     }]
    simSpaceSettings {maxAngularAcceleration=60, maxAngularVelocity=12, simSpaceAngularAlpha=0.75}
   }
   n30 = Anim Dynamics - Bone: dyn_hair_front_l_strand_05!(ComponentPose=@n29, Alpha=1, GravityOverride:Vector="0,0,-980", ExternalForce="0,0,0") -> (Pose) {
    angularDampingOverride=0.8
    angularSpringConstant=15000
    bAngularSpring=true
    bOverrideAngularDamping=true
    bUseGravityOverride=true
    simulationSpace=RootRelative
    alphaScaleBiasClamp {bInterpResult=true}
    boundBone {boneName=dyn_hair_front_l_strand_05}
    componentAppliedLinearAccClamp=<3000 3000 3000>
    componentLinearAccScale=<0.2 0.2 0.2>
    componentLinearVelScale=<0.2 0.2 0.2>
    gravityOverride {z=-980}
    physicsBodyDefinitions [{
      boundBone {boneName=dyn_hair_front_l_strand_05}
      boxExtents=<40 40 40>
      constraintSetup {angularLimitsMax=<0 10 10>, angularLimitsMin=<0 -10 -10>, angularTarget=<1 0 0>}
      localJointOffset=<1 0 0>
     }]
    simSpaceSettings {maxAngularAcceleration=60, maxAngularVelocity=12, simSpaceAngularAlpha=0.75}
   }
   n31 = Anim Dynamics - Bone: dyn_hair_front_l_strand_06!(ComponentPose=@n30, Alpha=1, GravityOverride:Vector="0,0,-980", ExternalForce="0,0,0") -> (Pose) {
    angularDampingOverride=0.8
    angularSpringConstant=15000
    bAngularSpring=true
    bOverrideAngularDamping=true
    bUseGravityOverride=true
    simulationSpace=RootRelative
    alphaScaleBiasClamp {bInterpResult=true}
    boundBone {boneName=dyn_hair_front_l_strand_06}
    componentAppliedLinearAccClamp=<3000 3000 3000>
    componentLinearAccScale=<0.2 0.2 0.2>
    componentLinearVelScale=<0.2 0.2 0.2>
    gravityOverride {z=-980}
    physicsBodyDefinitions [{
      boundBone {boneName=dyn_hair_front_l_strand_06}
      boxExtents=<40 40 40>
      constraintSetup {angularLimitsMax=<0 10 10>, angularLimitsMin=<0 -10 -10>, angularTarget=<1 0 0>}
      localJointOffset=<1 0 0>
     }]
    simSpaceSettings {maxAngularAcceleration=60, maxAngularVelocity=12, simSpaceAngularAlpha=0.75}
   }
   n32 = Anim Dynamics - Bone: dyn_hair_front_r_strand_01!(ComponentPose=@n31, Alpha=1, GravityOverride:Vector="0,0,-980", ExternalForce="0,0,0") -> (Pose) {
    angularDampingOverride=0.8
    angularSpringConstant=15000
    bAngularSpring=true
    bOverrideAngularDamping=true
    bUseGravityOverride=true
    simulationSpace=RootRelative
    alphaScaleBiasClamp {bInterpResult=true}
    boundBone {boneName=dyn_hair_front_r_strand_01}
    componentAppliedLinearAccClamp=<3000 3000 3000>
    componentLinearAccScale=<0.2 0.2 0.2>
    componentLinearVelScale=<0.2 0.2 0.2>
    gravityOverride {z=-980}
    physicsBodyDefinitions [{
      boundBone {boneName=dyn_hair_front_r_strand_01}
      boxExtents=<40 40 40>
      constraintSetup {angularLimitsMax=<0 10 10>, angularLimitsMin=<0 -10 -10>, angularTarget=<1 0 0>}
      localJointOffset=<1 0 0>
     }]
    simSpaceSettings {maxAngularAcceleration=60, maxAngularVelocity=12, simSpaceAngularAlpha=0.75}
   }
   n33 = Anim Dynamics - Bone: dyn_hair_front_r_strand_02!(ComponentPose=@n32, Alpha=1, GravityOverride:Vector="0,0,-980", ExternalForce="0,0,0") -> (Pose) {
    angularDampingOverride=0.8
    angularSpringConstant=15000
    bAngularSpring=true
    bOverrideAngularDamping=true
    bUseGravityOverride=true
    simulationSpace=RootRelative
    alphaScaleBiasClamp {bInterpResult=true}
    boundBone {boneName=dyn_hair_front_r_strand_02}
    componentAppliedLinearAccClamp=<3000 3000 3000>
    componentLinearAccScale=<0.2 0.2 0.2>
    componentLinearVelScale=<0.2 0.2 0.2>
    gravityOverride {z=-980}
    physicsBodyDefinitions [{
      boundBone {boneName=dyn_hair_front_r_strand_02}
      boxExtents=<40 40 40>
      constraintSetup {angularLimitsMax=<0 10 10>, angularLimitsMin=<0 -10 -10>, angularTarget=<1 0 0>}
      localJointOffset=<1 0 0>
     }]
    simSpaceSettings {maxAngularAcceleration=60, maxAngularVelocity=12, simSpaceAngularAlpha=0.75}
   }
   n34 = Anim Dynamics - Bone: dyn_hair_back_l_strand_01!(ComponentPose=@n33, Alpha=1, GravityOverride:Vector="0,0,-980", ExternalForce="0,0,0") -> (Pose) {
    angularDampingOverride=0.8
    angularSpringConstant=15000
    bAngularSpring=true
    bOverrideAngularDamping=true
    bUseGravityOverride=true
    simulationSpace=RootRelative
    alphaScaleBiasClamp {bInterpResult=true}
    boundBone {boneName=dyn_hair_back_l_strand_01}
    componentAppliedLinearAccClamp=<3000 3000 3000>
    componentLinearAccScale=<0.2 0.2 0.2>
    componentLinearVelScale=<0.2 0.2 0.2>
    gravityOverride {z=-980}
    physicsBodyDefinitions [{
      boundBone {boneName=dyn_hair_back_l_strand_01}
      boxExtents=<40 40 40>
      constraintSetup {angularLimitsMax=<0 10 10>, angularLimitsMin=<0 -10 -10>, angularTarget=<1 0 0>}
      localJointOffset=<1 0 0>
     }]
    simSpaceSettings {maxAngularAcceleration=60, maxAngularVelocity=12, simSpaceAngularAlpha=0.75}
   }
   n35 = Anim Dynamics - Bone: dyn_hair_back_l_strand_02!(ComponentPose=@n34, Alpha=1, GravityOverride:Vector="0,0,-980", ExternalForce="0,0,0") -> (Pose) {
    angularDampingOverride=0.8
    angularSpringConstant=15000
    bAngularSpring=true
    bOverrideAngularDamping=true
    bUseGravityOverride=true
    simulationSpace=RootRelative
    alphaScaleBiasClamp {bInterpResult=true}
    boundBone {boneName=dyn_hair_back_l_strand_02}
    componentAppliedLinearAccClamp=<3000 3000 3000>
    componentLinearAccScale=<0.2 0.2 0.2>
    componentLinearVelScale=<0.2 0.2 0.2>
    gravityOverride {z=-980}
    physicsBodyDefinitions [{
      boundBone {boneName=dyn_hair_back_l_strand_02}
      boxExtents=<40 40 40>
      constraintSetup {angularLimitsMax=<0 10 10>, angularLimitsMin=<0 -10 -10>, angularTarget=<1 0 0>}
      localJointOffset=<1 0 0>
     }]
    simSpaceSettings {maxAngularAcceleration=60, maxAngularVelocity=12, simSpaceAngularAlpha=0.75}
   }
   n36 = Anim Dynamics - Bone: dyn_hair_back_r_strand_01!(ComponentPose=@n35, Alpha=1, GravityOverride:Vector="0,0,-980", ExternalForce="0,0,0") -> (Pose) {
    angularDampingOverride=0.8
    angularSpringConstant=15000
    bAngularSpring=true
    bOverrideAngularDamping=true
    bUseGravityOverride=true
    simulationSpace=RootRelative
    alphaScaleBiasClamp {bInterpResult=true}
    boundBone {boneName=dyn_hair_back_r_strand_01}
    componentAppliedLinearAccClamp=<3000 3000 3000>
    componentLinearAccScale=<0.2 0.2 0.2>
    componentLinearVelScale=<0.2 0.2 0.2>
    gravityOverride {z=-980}
    physicsBodyDefinitions [{
      boundBone {boneName=dyn_hair_back_r_strand_01}
      boxExtents=<40 40 40>
      constraintSetup {angularLimitsMax=<0 10 10>, angularLimitsMin=<0 -10 -10>, angularTarget=<1 0 0>}
      localJointOffset=<1 0 0>
     }]
    simSpaceSettings {maxAngularAcceleration=60, maxAngularVelocity=12, simSpaceAngularAlpha=0.75}
   }
   n37 = Anim Dynamics - Bone: dyn_hair_back_r_strand_02!(ComponentPose=@n36, Alpha=1, GravityOverride:Vector="0,0,-980", ExternalForce="0,0,0") -> (Pose) {
    angularDampingOverride=0.8
    angularSpringConstant=15000
    bAngularSpring=true
    bOverrideAngularDamping=true
    bUseGravityOverride=true
    simulationSpace=RootRelative
    alphaScaleBiasClamp {bInterpResult=true}
    boundBone {boneName=dyn_hair_back_r_strand_02}
    componentAppliedLinearAccClamp=<3000 3000 3000>
    componentLinearAccScale=<0.2 0.2 0.2>
    componentLinearVelScale=<0.2 0.2 0.2>
    gravityOverride {z=-980}
    physicsBodyDefinitions [{
      boundBone {boneName=dyn_hair_back_r_strand_02}
      boxExtents=<40 40 40>
      constraintSetup {angularLimitsMax=<0 10 10>, angularLimitsMin=<0 -10 -10>, angularTarget=<1 0 0>}
      localJointOffset=<1 0 0>
     }]
    simSpaceSettings {maxAngularAcceleration=60, maxAngularVelocity=12, simSpaceAngularAlpha=0.75}
   }
   n38 = Anim Dynamics - Bone: dyn_hair_back_r_strand_03!(ComponentPose=@n37, Alpha=1, GravityOverride:Vector="0,0,-980", ExternalForce="0,0,0") -> (Pose) {
    angularDampingOverride=0.8
    angularSpringConstant=15000
    bAngularSpring=true
    bOverrideAngularDamping=true
    bUseGravityOverride=true
    simulationSpace=RootRelative
    alphaScaleBiasClamp {bInterpResult=true}
    boundBone {boneName=dyn_hair_back_r_strand_03}
    componentAppliedLinearAccClamp=<3000 3000 3000>
    componentLinearAccScale=<0.2 0.2 0.2>
    componentLinearVelScale=<0.2 0.2 0.2>
    gravityOverride {z=-980}
    physicsBodyDefinitions [{
      boundBone {boneName=dyn_hair_back_r_strand_03}
      boxExtents=<40 40 40>
      constraintSetup {angularLimitsMax=<0 10 10>, angularLimitsMin=<0 -10 -10>, angularTarget=<1 0 0>}
      localJointOffset=<1 0 0>
     }]
    simSpaceSettings {maxAngularAcceleration=60, maxAngularVelocity=12, simSpaceAngularAlpha=0.75}
   }
   n39 = Anim Dynamics - Bone: dyn_hair_back_mid_strand_01!(ComponentPose=@n38, Alpha=1, GravityOverride:Vector="0,0,-980", ExternalForce="0,0,0") -> (Pose) {
    angularDampingOverride=0.8
    angularSpringConstant=15000
    bAngularSpring=true
    bOverrideAngularDamping=true
    bUseGravityOverride=true
    simulationSpace=RootRelative
    alphaScaleBiasClamp {bInterpResult=true}
    boundBone {boneName=dyn_hair_back_mid_strand_01}
    componentAppliedLinearAccClamp=<3000 3000 3000>
    componentLinearAccScale=<0.2 0.2 0.2>
    componentLinearVelScale=<0.2 0.2 0.2>
    gravityOverride {z=-980}
    physicsBodyDefinitions [{
      boundBone {boneName=dyn_hair_back_mid_strand_01}
      boxExtents=<40 40 40>
      constraintSetup {angularLimitsMax=<0 10 10>, angularLimitsMin=<0 -10 -10>, angularTarget=<1 0 0>}
      localJointOffset=<1 0 0>
     }]
    simSpaceSettings {maxAngularAcceleration=60, maxAngularVelocity=12, simSpaceAngularAlpha=0.75}
   }
   n40 = Anim Dynamics - Bone: dyn_hair_back_mid_strand_02!(ComponentPose=@n39, Alpha=1, GravityOverride:Vector="0,0,-980", ExternalForce="0,0,0") -> (Pose) {
    angularDampingOverride=0.8
    angularSpringConstant=15000
    bAngularSpring=true
    bOverrideAngularDamping=true
    bUseGravityOverride=true
    simulationSpace=RootRelative
    alphaScaleBiasClamp {bInterpResult=true}
    boundBone {boneName=dyn_hair_back_mid_strand_02}
    componentAppliedLinearAccClamp=<3000 3000 3000>
    componentLinearAccScale=<0.2 0.2 0.2>
    componentLinearVelScale=<0.2 0.2 0.2>
    gravityOverride {z=-980}
    physicsBodyDefinitions [{
      boundBone {boneName=dyn_hair_back_mid_strand_02}
      boxExtents=<40 40 40>
      constraintSetup {angularLimitsMax=<0 10 10>, angularLimitsMin=<0 -10 -10>, angularTarget=<1 0 0>}
      localJointOffset=<1 0 0>
     }]
    simSpaceSettings {maxAngularAcceleration=60, maxAngularVelocity=12, simSpaceAngularAlpha=0.75}
   }
   Output Pose(Result=Component To Local(ComponentPose=@n40)) {layerGroup=None}
  ]}
 EventGraph {nodes [
   Event OnBeginSkydiving()
   > Set AnimDynamicsAlpha(AnimDynamicsAlpha:double=0) -> (Output_Get:double)
   > Set RBAN_Alpha(RBAN_Alpha:double=0.8) -> (Output_Get:double)
   Event OnEndSkydiving()
   > Set AnimDynamicsAlpha(AnimDynamicsAlpha=1)
   > Set RBAN_Alpha(RBAN_Alpha=1)
   n1 = Event OnCharacterCustomizationLoaded()
   > Set CharacterMesh(CharacterMesh:Skeletal Mesh Component Object Reference=@n1.CustomBaseMesh) -> (Output_Get:Skeletal Mesh Component Object Reference)
   > Set HeadMesh(HeadMesh:Skeletal Mesh Component Budgeted Object Reference=Get Array Item with EFort Custom Part Type(Array=@n1.CustomMeshParts, PartType=Head)) -> (Output_Get:Skeletal Mesh Component Budgeted Object Reference)
   > Set isWearingPack(isWearingPack:bool=CheckForBackbling()) -> (Output_Get:bool)
   Event OnEnterGrindRail() {is=title K2Node_Event}
   > Set AnimDynamicsAlpha(AnimDynamicsAlpha=0)
   n2 = Event BlueprintProcessAnimInputEvent()
   > n3 = Branch(Condition=IsNamed(self=@n2.AnimInputEvent, InEventName=aie.Player.Zipline.Entered))
   > Set AnimDynamicsAlpha(AnimDynamicsAlpha=0)
   Branch(execute=@n3.else, Condition=IsNamed(self=@n2.AnimInputEvent, InEventName=aie.Player.Zipline.Exited))
   > Set AnimDynamicsAlpha(AnimDynamicsAlpha=1)
   Event OnExitGrindRail() {is=title K2Node_Event}
   > Set AnimDynamicsAlpha(AnimDynamicsAlpha=1)
   Event BlueprintUpdateAnimation()
   > Set arm_r_up_SPR_alpha(arm_r_up_SPR_alpha:double=Remap Float(Input=GetCurveValue(CurveName=arm_r_up_SPR), OldRange=(X=0,Y=1), NewRange=(X=0,Y=0.8), Clamp=true) -> (Output=~)) -> (Output_Get:double)
   > Set arm_l_up_SPR_alpha(arm_l_up_SPR_alpha:double=Remap Float(Input=GetCurveValue(CurveName=arm_l_up_SPR), OldRange=(X=0,Y=1), NewRange=(X=0,Y=0.8), Clamp=true) -> (Output=~)) -> (Output_Get:double)
   > n4 = "Gravity Override State Function (Parameter Asset) (Nativized)"(
    CharacterPartAnimInstance=Self-Reference()
    GameTimeInSeconds=GetGameTimeInSeconds()
    Parameters=&/M_MED_FloralMane_Sand_FA_Parameters
   )
   > Set Gravity Override(Gravity Override=@n4)
   > Set RBAN Settings Out(RBAN Settings Out=@n4.OutRigidBodyAnimNodeParameters)
   > n5 = Branch(Condition=Get bIsSkyDiving())
   > n6 = Branch(Condition=Get bIsSkydiveDiveMode())
   > Set skydiveAlpha(skydiveAlpha:double=0) -> (Output_Get:double)
   n7 = Branch(execute=@n6.else, Condition=Get bIsParachuteOpened())
   > Set skydiveAlpha(skydiveAlpha=0)
   > Set RBAN_Alpha(RBAN_Alpha=0.9)
   Set skydiveAlpha(execute=@n7.else, skydiveAlpha=1)
   > Set RBAN_Alpha(RBAN_Alpha=1)
   Set skydiveAlpha(execute=@n5.else, skydiveAlpha=0)
  ]}
]
variables [
 AnimDynamicsAlpha {type=number, default=1}
 arm_l_up_SPR_alpha {type=number}
 arm_r_up_SPR_alpha {type=number}
 CharacterMesh {type=SkeletalMeshComponent}
 Gravity Override {type=Vector}
 HeadMesh {type=SkeletalMeshComponentBudgeted}
 isWearingPack {type=boolean}
 packConstraintAlpha {type=number, default=0.25}
 packHairAlpha {type=number, default=0.3}
 RBAN Settings Out {type=FortRigidBodyAnimNodeParameters}
 RBAN_Alpha {type=number, default=1}
 skydiveAlpha {type=number}
]
